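# Control Rig for the M_LRG_BitterSweet player body, adding pose-driven corrective deformation.
# A shared DeformRig function runs first, then bone offsets adjust accessories like spine metal, sneakers and grenade belt.
# Pose readers, bone angles, curves and state variables weight these offsets, some Kalman-smoothed.
# Without gauntlets equipped, an arm correction solves PBIK on proxy bones, twists them by pose and hand distance, then constrains real arms back.
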ControlRigBlueprint /BRCosmetics/Characters/Player/Male/Large/Bodies/M_LRG_BitterSweet/Meshes/M_LRG_BitterSweet_CtrlRig
graphs [
 New Function_1_ContainedGraph {nodes [
   RigVMFunctionEntryNode()
   > RigVMFunction_Sequence_1 = RigVMAggregateNode()
   RigVMFunction_Sequence_2 = RigVMAggregateNode(ExecuteContext=@RigVMFunction_Sequence_1.A)
   n1 = RigVMDispatchNode(Array=((Type=Bone,Name="upperarm_l"),(Type=Bone,Name="lowerarm_l"),(Type=Bone,Name="hand_l")), ExecuteContext=@RigVMFunction_Sequence_2.A) {template="DISPATCH_RigVMDispatch_ArrayIterator(in Array,out Element,out Index,out Count,out Ratio)"}
   > &RigUnit_SetTransform(
    Weight=1
    bInitial=False
    Item=RigVMDispatchNode(Array=((Type=Bone,Name="p_deform_upperarm_l"),(Type=Bone,Name="p_deform_lowerarm_l"),(Type=Bone,Name="p_deform_hand_l")), Index=@n1.Index) {template="DISPATCH_RigVMDispatch_ArrayGetAtIndex(in Array,in Index,out Element)"}.Element
    Value=&RigUnit_GetTransform(bInitial=False, Item=@n1.Element).Transform
   )
   n2 = /Script/PBIK.RigUnit_PBIK(
    BoneSettings=((Bone="deform_hand_l",RotationStiffness=0,PositionStiffness=0,X=Free,MinX=0,MaxX=0,Y=Free,MinY=0,MaxY=0,Z=Free,MinZ=0,MaxZ=0,bUsePreferredAngles=true,PreferredAngles=(X=1,Y=0,Z=0)))
    Debug=(DrawScale=1,bDrawDebug=false)
    Effectors=((Bone="hand_l",Transform=(Rotation=(X=0,Y=0,Z=0,W=1),Translation=(X=0,Y=0,Z=0),Scale3D=(X=1,Y=1,Z=1)),PositionAlpha=1,RotationAlpha=1,StrengthAlpha=1,PullChainAlpha=1,PinRotation=1,ChainDepth=0))
    ExcludedBones=()
    Root=p_deform_upperarm_l
    Settings="(Iterations=20,SubIterations=0,MassMultiplier=1,bAllowStretch=false,RootBehavior=PrePull,PrePullRootSettings=(RotationAlpha=0,RotationAlphaX=1,RotationAlphaY=1,RotationAlphaZ=1,PositionAlpha=1,PositionAlphaX=1,PositionAlphaY=1,PositionAlphaZ=1),GlobalPullC ...(truncated)"
    ExecutePin=@n1.Completed
   ) {is=struct RigVMUnitNode}
   RigVMFunction_ControlFlowBranch = &RigVMFunction_ControlFlowBranch(Condition=False, ExecuteContext=@n2.ExecutePin)
   n3 = &RigUnit_SphericalPoseReader(
    ActiveRegionScaleFactors=(PositiveWidth=2,NegativeWidth=0.5,PositiveHeight=1,NegativeHeight=1)
    ActiveRegionSize=0.1
    Debug=(bDrawDebug=false,bDraw2D=False,bDrawLocalAxes=False,DebugScale=25,DebugSegments=20,DebugThickness=0.25)
    DriverAxis=(X=1,Y=0,Z=0)
    DriverItem=(Type=Bone,Name="upperarm_l")
    FalloffRegionScaleFactors=(PositiveWidth=1,NegativeWidth=1,PositiveHeight=2,NegativeHeight=2)
    FalloffSize=0.2
    FlipHeightScaling=False
    FlipWidthScaling=False
    OptionalParentItem=(Type=Bone,Name="spine_03")
    RotationOffset=(X=0,Y=-145,Z=0)
    ExecutePin=@RigVMFunction_ControlFlowBranch.False
   )
   > ApplyFBIKChainCorrection = RigVMFunctionReferenceNode(End=(Type=Bone,Name="p_deform_hand_l"), QuatAmount=-0.4, Start=(Type=Bone,Name="p_deform_upperarm_l"), Weight=@n3.OutputParam)
   > Get Normalized Distance Between_1 = RigVMFunctionReferenceNode(Debug=false, Max Distance=40, Min Distance=10, Source=(Type=Bone,Name="hand_l"), Target=(Type=Bone,Name="pelvis"))
   > ApplyFBIKChainCorrection_1 = RigVMFunctionReferenceNode(
    End=(Type=Bone,Name="p_deform_hand_l")
    QuatAmount=-0.8
    Start=(Type=Bone,Name="p_deform_upperarm_l")
    Weight=@"Get Normalized Distance Between_1".Result
   )
   n4 = RigVMDispatchNode(Array=((Type=Bone,Name="upperarm_r"),(Type=Bone,Name="lowerarm_r"),(Type=Bone,Name="hand_r")), ExecuteContext=@RigVMFunction_Sequence_2.B) {template="DISPATCH_RigVMDispatch_ArrayIterator(in Array,out Element,out Index,out Count,out Ratio)"}
   > &RigUnit_SetTransform(
    Weight=1
    bInitial=False
    Item=RigVMDispatchNode(Array=((Type=Bone,Name="p_deform_upperarm_r"),(Type=Bone,Name="p_deform_lowerarm_r"),(Type=Bone,Name="p_deform_hand_r")), Index=@n4.Index) {template="DISPATCH_RigVMDispatch_ArrayGetAtIndex(in Array,in Index,out Element)"}.Element
    Value=&RigUnit_GetTransform(bInitial=False, Item=@n4.Element).Transform
   )
   n5 = &RigUnit_PBIK(
    BoneSettings="((Bone=\"deform_hand_r\",RotationStiffness=0,PositionStiffness=0,X=Free,MinX=0,MaxX=0,Y=Free,MinY=0,MaxY=0,Z=Free,MinZ=0,MaxZ=0,bUsePreferredAngles=false,PreferredAngles=(X=1,Y=0,Z=0)),(Bone=\"deform_upperarm_l\",RotationStiffness=0,PositionStiffness=0,X=Free, ...(truncated)"
    Debug=(DrawScale=3,bDrawDebug=true)
    Effectors="((Bone=\"hand_r\",Transform=(Rotation=(X=0,Y=0,Z=0,W=1),Translation=(X=0,Y=0,Z=0),Scale3D=(X=1,Y=1,Z=1)),PositionAlpha=1,RotationAlpha=1,StrengthAlpha=1,PullChainAlpha=1,PinRotation=1,ChainDepth=0),(Bone=\"upperarm_r\",Transform=(Rotation=(X=0,Y=0,Z=0,W=1),Tra ...(truncated)"
    ExcludedBones=()
    Root=p_deform_upperarm_r
    Settings="(Iterations=20,SubIterations=0,MassMultiplier=1,bAllowStretch=false,RootBehavior=PrePull,PrePullRootSettings=(RotationAlpha=0,RotationAlphaX=1,RotationAlphaY=1,RotationAlphaZ=1,PositionAlpha=1,PositionAlphaX=1,PositionAlphaY=1,PositionAlphaZ=1),GlobalPullC ...(truncated)"
    ExecutePin=@n4.Completed
   )
   n6 = &RigVMFunction_ControlFlowBranch(Condition=False, ExecuteContext=@n5.ExecutePin)
   n7 = &RigUnit_SphericalPoseReader(
    ActiveRegionScaleFactors=(PositiveWidth=2,NegativeWidth=0.5,PositiveHeight=1,NegativeHeight=1)
    ActiveRegionSize=0.1
    Debug=(bDrawDebug=false,bDraw2D=False,bDrawLocalAxes=False,DebugScale=25,DebugSegments=20,DebugThickness=0.25)
    DriverAxis=(X=-1,Y=0,Z=0)
    DriverItem=(Type=Bone,Name="upperarm_r")
    FalloffRegionScaleFactors=(PositiveWidth=1,NegativeWidth=1,PositiveHeight=2,NegativeHeight=2)
    FalloffSize=0.2
    FlipHeightScaling=False
    FlipWidthScaling=False
    OptionalParentItem=(Type=Bone,Name="spine_03")
    RotationOffset=(X=0,Y=35,Z=0)
    ExecutePin=@n6.False
   )
   > ApplyFBIKChainCorrection_2 = RigVMFunctionReferenceNode(End=(Type=Bone,Name="p_deform_hand_r"), QuatAmount=1, Start=(Type=Bone,Name="p_deform_upperarm_r"), Weight=@n7.OutputParam)
   > Get Normalized Distance Between_1_1 = RigVMFunctionReferenceNode(Debug=false, Max Distance=35, Min Distance=1, Source=(Type=Bone,Name="hand_r"), Target=(Type=Bone,Name="spine_04"))
   > ApplyFBIKChainCorrection_3 = RigVMFunctionReferenceNode(
    End=(Type=Bone,Name="p_deform_hand_r")
    QuatAmount=0.4
    Start=(Type=Bone,Name="p_deform_upperarm_r")
    Weight=@"Get Normalized Distance Between_1_1".Result
   )
   &RigUnit_ParentConstraint(
    AdvancedSettings=(InterpolationType=Average,RotationOrderForFilter=XZY)
    Child=(Type=Bone,Name="upperarm_l")
    Parents=((Item=(Type=Bone,Name="p_deform_upperarm_l"),Weight=1))
    Weight=1
    ExecutePin=@RigVMFunction_Sequence_1.C
   )
   > &RigUnit_ParentConstraint(
    AdvancedSettings=(InterpolationType=Average,RotationOrderForFilter=XZY)
    Child=(Type=Bone,Name="lowerarm_l")
    Parents=((Item=(Type=Bone,Name="p_deform_lowerarm_l"),Weight=1))
    Weight=1
   )
   > &RigUnit_ParentConstraint(
    AdvancedSettings=(InterpolationType=Average,RotationOrderForFilter=XZY)
    Child=(Type=Bone,Name="hand_l")
    Parents=((Item=(Type=Bone,Name="p_deform_hand_l"),Weight=1))
    Weight=1
   )
   > &RigUnit_ParentConstraint(
    AdvancedSettings=(InterpolationType=Average,RotationOrderForFilter=XZY)
    Child=(Type=Bone,Name="upperarm_r")
    Parents=((Item=(Type=Bone,Name="p_deform_upperarm_r"),Weight=1))
    Weight=1
   )
   > &RigUnit_ParentConstraint(
    AdvancedSettings=(InterpolationType=Average,RotationOrderForFilter=XZY)
    Child=(Type=Bone,Name="lowerarm_r")
    Parents=((Item=(Type=Bone,Name="p_deform_lowerarm_r"),Weight=1))
    Weight=1
   )
   > &RigUnit_ParentConstraint(
    AdvancedSettings=(InterpolationType=Average,RotationOrderForFilter=XZY)
    Child=(Type=Bone,Name="hand_r")
    Parents=((Item=(Type=Bone,Name="p_deform_hand_r"),Weight=1))
    Weight=1
   )
   RigVMFunctionReturnNode(ExecuteContext=@RigVMFunction_Sequence_1.D)
   RigUnit_GetTransform = &RigUnit_GetTransform(Item=(Type=Bone,Name="hand_l"), bInitial=False)
   RigUnit_SetTransform = &RigUnit_SetTransform(
    Item=(Type=Bone,Name="p_deform_hand_l")
    Weight=1
    bInitial=False
    Value=@RigUnit_GetTransform.Transform
    ExecutePin=@RigVMFunction_Sequence_1.B
   )
   > &RigUnit_SetTransform(
    Item=(Type=Bone,Name="p_deform_hand_r")
    Weight=1
    bInitial=False
    Value=&RigUnit_GetTransform(Item=(Type=Bone,Name="hand_r"), bInitial=False).Transform
   )
   IKTwoBone = RigVMFunctionReferenceNode(
    Bone A=(Type=Bone,Name="upperarm_l")
    Bone B=(Type=Bone,Name="lowerarm_l")
    Bone C=(Type=Bone,Name="hand_l")
    Enable Stretch=true
    End Ctrl=(Type=Control,Name="hand_l_ctrl")
    ExecuteContext=()
    Line Thickness=0.25
    Pole Vector Node=(Type=Control,Name="PoleVectorl_l")
    Primary Axis=(X=1,Y=0,Z=0)
    Secondary Axis=(X=0,Y=-1,Z=0)
    Stretch Maximum Ratio=1.25
    Stretch Start Ratio=0.75
    debug=false
    scale=(X=0,Y=0,Z=0)
   )
   > IKTwoBone_1 = RigVMFunctionReferenceNode(
    Bone A=(Type=Bone,Name="upperarm_r")
    Bone B=(Type=Bone,Name="lowerarm_r")
    Bone C=(Type=Bone,Name="hand_r")
    Enable Stretch=false
    End Ctrl=(Type=Control,Name="hand_r_ctrl")
    Line Thickness=0.25
    Pole Vector Node=(Type=Control,Name="PoleVector_r")
    Primary Axis=(X=-1,Y=0,Z=0)
    Secondary Axis=(X=0,Y=1,Z=0)
    Stretch Maximum Ratio=1.25
    Stretch Start Ratio=0.75
    debug=false
    scale=(X=0,Y=0,Z=0)
   )
   Get Normalized Distance Between_1_2 = RigVMFunctionReferenceNode(Debug=true, ExecuteContext=(), Max Distance=35, Min Distance=1, Source=(Type=Bone,Name="hand_r"), Target=(Type=Bone,Name="pelvis"))
   RerouteNode_4_Constant = RigVMDispatchNode(Value=False) {template="DISPATCH_RigVMDispatch_Constant(io Value)"}
   RigVMRerouteNode(Value=@RerouteNode_4_Constant.Value)
   RigVMRerouteNode(Value=@RerouteNode_4_Constant.Value)
   &RigUnit_GetTransform(Item=(Type=Bone,Name="upperarm_r"), bInitial=False)
   &RigUnit_GetTransform(Item=(Type=Control,Name="hand_r_ctrl"), bInitial=False)
   RigVMVariableNode() {variable=l_arm_bones}
  ]}
 New Function_ContainedGraph {nodes [
   n1 = RigVMFunctionEntryNode()
   n2 = &RigVMFunction_Sequence(ExecuteContext=@n1.ExecuteContext)
   ArmIKCorrection = RigVMFunctionReferenceNode(disable_arms_modifications=1, ExecuteContext=@n2.A)
   After correct = RigVMFunctionReferenceNode(ExecuteContext=@n2.B)
   > RigVMFunctionReturnNode()
  ]}
 New Function_ContainedGraph {nodes [
   RigVMFunctionEntryNode()
   > RigVMFunction_Sequence_1 = RigVMAggregateNode()
   &RigUnit_ParentConstraint(
    AdvancedSettings=(InterpolationType=Average,RotationOrderForFilter=XZY)
    Child=(Type=Bone,Name="lowerarm_twist_01_l")
    Parents=((Item=(Type=Bone,Name="p_deform_lowerarm_twist_01_l"),Weight=1))
    Weight=1
    ExecutePin=@RigVMFunction_Sequence_1.E
   )
   > &RigUnit_ParentConstraint(
    AdvancedSettings=(InterpolationType=Average,RotationOrderForFilter=XZY)
    Child=(Type=Bone,Name="lowerarm_twist_02_l")
    Parents=((Item=(Type=Bone,Name="p_deform_lowerarm_twist_02_l"),Weight=1))
    Weight=1
   )
   > RigVMFunctionReturnNode()
   n1 = /Script/RigVM.RigVMFunction_MathQuaternionToAxisAndAngle(Value=&RigVMFunction_MathQuaternionSwingTwist(
     TwistAxis=(X=1,Y=0,Z=0)
     Input=&RigUnit_GetTransform(Item=(Type=Bone,Name="p_deform_upperarm_r"), Space=LocalSpace, bInitial=False).Transform.Rotation
    ).Twist) {is=struct RigVMUnitNode, template="ToAxisAndAngle::Execute(in Value,out Axis,out Angle)"}
   &RigUnit_SetRotation(
    Item=(Type=Bone,Name="p_deform_upperarm_twist_01_r")
    Weight=1
    bInitial=False
    Value=/Script/RigVM.RigVMFunction_MathQuaternionFromAxisAndAngle(Angle=&RigVMFunction_MathDoubleMul(B=-0.6, A=@n1.Angle).Result, Axis=@n1.Axis) {is=struct RigVMUnitNode, template="FromAxisAndAngle::Execute(in Axis,in Angle,out Result)"}.Result
    ExecutePin=@RigVMFunction_Sequence_1.B
   )
   > &RigUnit_SetRotation(
    Item=(Type=Bone,Name="p_deform_upperarm_twist_02_r")
    Weight=1
    bInitial=False
    Value=&RigVMFunction_MathQuaternionFromAxisAndAngle(Angle=&RigVMFunction_MathDoubleMul(B=-0.2, A=@n1.Angle).Result, Axis=@n1.Axis) {template="FromAxisAndAngle::Execute(in Axis,in Angle,out Result)"}.Result
   )
   > &RigUnit_SetRotation(
    Item=(Type=Bone,Name="p_deform_lowerarm_twist_01_r")
    Weight=1
    bInitial=False
    Value=&RigVMFunction_MathQuaternionSwingTwist(
     TwistAxis=(X=1,Y=0,Z=0)
     Input=&RigUnit_GetTransform(Item=(Type=Bone,Name="lowerarm_twist_01_r"), Space=LocalSpace, bInitial=False).Transform.Rotation
    ).Twist
   )
   > &RigUnit_SetRotation(
    Item=(Type=Bone,Name="p_deform_lowerarm_twist_02_r")
    Weight=0.3
    bInitial=False
    Value=&RigVMFunction_MathQuaternionSwingTwist(
     TwistAxis=(X=1,Y=0,Z=0)
     Input=&RigUnit_GetTransform(Item=(Type=Bone,Name="lowerarm_twist_02_r"), Space=LocalSpace, bInitial=False).Transform.Rotation
    ).Swing
   )
   RigUnit_GetTransform = &RigUnit_GetTransform(Item=(Type=Bone,Name="p_deform_upperarm_l"), Space=LocalSpace, bInitial=False)
   RigVMFunction_MathQuaternionSwingTwist = &RigVMFunction_MathQuaternionSwingTwist(TwistAxis=(X=-1,Y=0,Z=0), Input=@RigUnit_GetTransform.Transform.Rotation)
   RigVMFunction_MathQuaternionToAxisAndAngle = &RigVMFunction_MathQuaternionToAxisAndAngle(Value=@RigVMFunction_MathQuaternionSwingTwist.Twist) {template="ToAxisAndAngle::Execute(in Value,out Axis,out Angle)"}
   RigVMFunction_MathQuaternionFromAxisAndAngle = &RigVMFunction_MathQuaternionFromAxisAndAngle(
    Angle=&RigVMFunction_MathDoubleMul(B=-0.6, A=@RigVMFunction_MathQuaternionToAxisAndAngle.Angle).Result
    Axis=@RigVMFunction_MathQuaternionToAxisAndAngle.Axis
   ) {template="FromAxisAndAngle::Execute(in Axis,in Angle,out Result)"}
   &RigUnit_SetRotation(
    Item=(Type=Bone,Name="p_deform_upperarm_twist_01_l")
    Weight=1
    bInitial=False
    Value=@RigVMFunction_MathQuaternionFromAxisAndAngle.Result
    ExecutePin=@RigVMFunction_Sequence_1.A
   )
   > &RigUnit_SetRotation(
    Item=(Type=Bone,Name="p_deform_upperarm_twist_02_l")
    Weight=1
    bInitial=False
    Value=&RigVMFunction_MathQuaternionFromAxisAndAngle(
     Angle=&RigVMFunction_MathDoubleMul(B=-0.2, A=@RigVMFunction_MathQuaternionToAxisAndAngle.Angle).Result
     Axis=@RigVMFunction_MathQuaternionToAxisAndAngle.Axis
    ) {template="FromAxisAndAngle::Execute(in Axis,in Angle,out Result)"}.Result
   )
   n2 = &RigUnit_GetTransform(Item=(Type=Bone,Name="lowerarm_twist_01_l"), Space=LocalSpace, bInitial=False)
   > &RigUnit_SetRotation(Item=(Type=Bone,Name="p_deform_lowerarm_twist_01_l"), Weight=1, bInitial=False, Value=@n2.Transform.Rotation)
   n3 = &RigUnit_GetTransform(Item=(Type=Bone,Name="lowerarm_twist_02_l"), Space=LocalSpace, bInitial=False)
   > &RigUnit_SetRotation(Item=(Type=Bone,Name="p_deform_lowerarm_twist_02_l"), Weight=0.3, bInitial=False, Value=@n3.Transform.Rotation)
   &RigUnit_ParentConstraint(
    AdvancedSettings=(InterpolationType=Average,RotationOrderForFilter=XZY)
    Child=(Type=Bone,Name="None")
    ExecutePin=()
    Parents=((Item=(Type=Bone,Name="None"),Weight=1))
    Weight=1
   )
   &RigUnit_ParentConstraint(
    AdvancedSettings=(InterpolationType=Average,RotationOrderForFilter=XZY)
    Child=(Type=Bone,Name="upperarm_l")
    ExecutePin=()
    Parents=((Item=(Type=Bone,Name="p_deform_upperarm_l"),Weight=1))
    Weight=1
   )
   &RigUnit_ParentConstraint(
    AdvancedSettings=(InterpolationType=Average,RotationOrderForFilter=XZY)
    Child=(Type=Bone,Name="hand_l")
    ExecutePin=()
    Parents=((Item=(Type=Bone,Name="p_deform_hand_l"),Weight=1))
    Weight=1
   )
   &RigUnit_ParentConstraint(
    AdvancedSettings=(InterpolationType=Average,RotationOrderForFilter=XZY)
    Child=(Type=Bone,Name="upperarm_r")
    ExecutePin=()
    Parents=((Item=(Type=Bone,Name="p_deform_upperarm_r"),Weight=1))
    Weight=1
   )
   RigVMRerouteNode(Value=@RigVMFunction_Sequence_1.D)
   RigVMRerouteNode(Value=@RigVMFunction_Sequence_1.C)
   n4 = &RigUnit_GetTransform(Item=(Type=Bone,Name="clavicle_l"), Space=LocalSpace, bInitial=False)
   n5 = &RigUnit_GetTransform(Item=(Type=Bone,Name="clavicle_l"), Space=LocalSpace, bInitial=true)
   n6 = &RigVMFunction_MathQuaternionToAxisAndAngle(Value=&RigVMFunction_MathQuaternionSwingTwist(
     TwistAxis=(X=1,Y=0,Z=0)
     Input=&RigVMFunction_MathTransformMakeRelative(Parent=@n4.Transform, Global=@n5.Transform).Local.Rotation
    ).Twist) {template="ToAxisAndAngle::Execute(in Value,out Axis,out Angle)"}
   &RigUnit_SetRotation(
    ExecutePin=()
    Item=(Type=Bone,Name="deform_clavicle_l")
    Weight=1
    bInitial=False
    Value=&RigVMFunction_MathQuaternionFromAxisAndAngle(Angle=&RigVMFunction_MathDoubleMul(B=-1, A=@n6.Angle).Result, Axis=@n6.Axis) {template="FromAxisAndAngle::Execute(in Axis,in Angle,out Result)"}.Result
   )
   n7 = &RigUnit_GetTransform(Item=(Type=Bone,Name="clavicle_r"), Space=LocalSpace, bInitial=true)
   n8 = &RigUnit_GetTransform(Item=(Type=Bone,Name="clavicle_r"), Space=LocalSpace, bInitial=False)
   n9 = &RigVMFunction_MathQuaternionToAxisAndAngle(Value=&RigVMFunction_MathQuaternionSwingTwist(
     TwistAxis=(X=1,Y=0,Z=0)
     Input=&RigVMFunction_MathTransformMakeRelative(Global=@n7.Transform, Parent=@n8.Transform).Local.Rotation
    ).Twist) {template="ToAxisAndAngle::Execute(in Value,out Axis,out Angle)"}
   &RigUnit_SetRotation(
    ExecutePin=()
    Item=(Type=Bone,Name="deform_clavicle_r")
    Weight=1
    bInitial=False
    Value=&RigVMFunction_MathQuaternionFromAxisAndAngle(Angle=&RigVMFunction_MathDoubleMul(B=-1, A=@n9.Angle).Result, Axis=@n9.Axis) {template="FromAxisAndAngle::Execute(in Axis,in Angle,out Result)"}.Result
   )
   n10 = &RigVMFunction_MathQuaternionToAxisAndAngle(Value=&RigVMFunction_MathQuaternionSwingTwist(
     TwistAxis=(X=1,Y=0,Z=0)
     Input=&RigUnit_GetTransform(Item=(Type=Bone,Name="hand_l"), Space=LocalSpace, bInitial=false).Transform.Rotation
    ).Twist) {template="ToAxisAndAngle::Execute(in Value,out Axis,out Angle)"}
   Multiply = &RigVMFunction_MathDoubleMul(B=0.4, A=@n10.Angle)
   &RigUnit_SetRotation(
    ExecutePin=()
    Item=(Type=Bone,Name="p_deform_lowerarm_twist_01_l")
    Weight=1
    bInitial=False
    bPropagateToChildren=true
    Value=&RigVMFunction_MathQuaternionFromAxisAndAngle(Angle=@Multiply.Result, Axis=@n10.Axis) {template="FromAxisAndAngle::Execute(in Axis,in Angle,out Result)"}.Result
   )
   n11 = &RigVMFunction_MathQuaternionToAxisAndAngle(Value=&RigVMFunction_MathQuaternionSwingTwist(
     TwistAxis=(X=1,Y=0,Z=0)
     Input=&RigUnit_GetTransform(Item=(Type=Bone,Name="hand_r"), Space=LocalSpace, bInitial=false).Transform.Rotation
    ).Twist) {template="ToAxisAndAngle::Execute(in Value,out Axis,out Angle)"}
   &RigVMFunction_MathQuaternionFromAxisAndAngle(Angle=&RigVMFunction_MathDoubleMul(B=0.4, A=@n11.Angle).Result, Axis=@n11.Axis) {template="FromAxisAndAngle::Execute(in Axis,in Angle,out Result)"}
   &RigVMFunction_MathQuaternionSwingTwist(TwistAxis=(X=1,Y=0,Z=0), Input=@n2.Transform.Rotation)
   &RigVMFunction_MathQuaternionSwingTwist(TwistAxis=(X=1,Y=0,Z=0), Input=@n3.Transform.Rotation)
  ]}
 New Function_ContainedGraph {nodes [
   n1 = RigVMFunctionEntryNode()
   n2 = &RigUnit_GetTransform(bInitial=False, Item=@n1.Start)
   Subtract = /Script/RigVM.RigVMFunction_MathVectorSub(B=@n2.Transform.Translation, A=&RigUnit_GetTransform(bInitial=False, Item=@n1.End).Transform.Translation) {is=struct RigVMUnitNode, template="Subtract::Execute(in A,in B,out Result)"}
   Unit = /Script/RigVM.RigVMFunction_MathVectorUnit(Value=@Subtract.Result) {is=struct RigVMUnitNode, template="Unit::Execute(in Value,out Result)"}
   RigVMFunction_MathQuaternionFromAxisAndAngle = &RigVMFunction_MathQuaternionFromAxisAndAngle(Axis=@Unit.Result, Angle=@n1.QuatAmount) {template="FromAxisAndAngle::Execute(in Axis,in Angle,out Result)"}
   Multiply = /Script/RigVM.RigVMFunction_MathQuaternionMul(B=@n2.Transform.Rotation, A=@RigVMFunction_MathQuaternionFromAxisAndAngle.Result) {is=struct RigVMUnitNode, template="Multiply::Execute(in A,in B,out Result)"}
   > &RigUnit_SetRotation(Space=GlobalSpace, bInitial=False, Item=@n1.Start, Weight=@n1.Weight, Value=@Multiply.Result)
   > RigVMFunctionReturnNode()
  ]}
 RigVMFunctionLibrary {nodes [
   After correct = RigVMCollapseNode(ExecuteContext=())
   ApplyFBIKChainCorrection = RigVMCollapseNode(End=(Type=None,Name="None"), ExecuteContext=(), Start=(Type=None,Name="None"))
   ArmIKCorrection = RigVMCollapseNode(ExecuteContext=())
   IK correction BitterSweet = RigVMCollapseNode(ExecuteContext=())
  ]}
 RigVMFunction_Sequence_1_1_ContainedGraph {nodes [
   n1 = RigVMFunctionEntryNode()
   n2 = &RigVMFunction_Sequence(ExecuteContext=@n1.ExecuteContext)
   n3 = &RigVMFunction_Sequence(ExecuteContext=@n2.B)
   RigVMFunctionReturnNode(A=@n2.A, B=@n3.A, C=@n3.B)
  ]}
 RigVMFunction_Sequence_1_ContainedGraph {nodes [
   n1 = RigVMFunctionEntryNode()
   RigVMFunction_Sequence = &RigVMFunction_Sequence(ExecuteContext=@n1.ExecuteContext)
   n2 = &RigVMFunction_Sequence(ExecuteContext=@RigVMFunction_Sequence.B)
   n3 = &RigVMFunction_Sequence(ExecuteContext=@n2.B)
   n4 = &RigVMFunction_Sequence(ExecuteContext=@n3.B)
   n5 = &RigVMFunction_Sequence(ExecuteContext=@n4.B)
   RigVMFunctionReturnNode(A=@RigVMFunction_Sequence.A, B=@n2.A, C=@n3.A, D=@n4.A, E=@n5.A, F=@n5.B)
  ]}
 RigVMFunction_Sequence_1_ContainedGraph {nodes [
   n1 = RigVMFunctionEntryNode()
   RigVMFunction_Sequence = &RigVMFunction_Sequence(ExecuteContext=@n1.ExecuteContext)
   n2 = &RigVMFunction_Sequence(ExecuteContext=@RigVMFunction_Sequence.B)
   n3 = &RigVMFunction_Sequence(ExecuteContext=@n2.B)
   RigVMFunctionReturnNode(A=@RigVMFunction_Sequence.A, B=@n2.A, C=@n3.A, D=@n3.B)
  ]}
 RigVMFunction_Sequence_1_ContainedGraph {nodes [
   n1 = RigVMFunctionEntryNode()
   RigVMFunction_Sequence = &RigVMFunction_Sequence(ExecuteContext=@n1.ExecuteContext)
   n2 = &RigVMFunction_Sequence(ExecuteContext=@RigVMFunction_Sequence.B)
   n3 = &RigVMFunction_Sequence(ExecuteContext=@n2.B)
   n4 = &RigVMFunction_Sequence(ExecuteContext=@n3.B)
   RigVMFunctionReturnNode(A=@RigVMFunction_Sequence.A, B=@n2.A, C=@n3.A, D=@n4.A, E=@n4.B)
  ]}
 RigVMModel {nodes [
   &RigUnit_ModifyTransforms(
    ExecutePin=()
    ItemToModify=((Item=(Type=Bone,Name="Spine_metal_corr_up"),Transform=(Rotation=(X=0,Y=0,Z=0.0262,W=0.9997),Translation=(X=0,Y=-0.5,Z=0),Scale3D=(X=1,Y=1,Z=1))))
    Weight=1
    WeightMinimum=0
   )
   > Get Delta Ref = RigVMFunctionReferenceNode(Item=(Type=Bone,Name="pelvis"))
   > &RigUnit_ModifyTransforms(
    ItemToModify=((Item=(Type=Bone,Name="Spine_metal_corr_up"),Transform=(Rotation=(X=0,Y=0,Z=0.0872,W=0.9962),Translation=(X=0,Y=0,Z=0),Scale3D=(X=1,Y=1,Z=1))))
    WeightMinimum=0
    Weight=&RigVMFunction_MathDoubleRemap(SourceMaximum=80, SourceMinimum=50, TargetMaximum=1, TargetMinimum=0, Value=@"Get Delta Ref".Result.Rotation.Yaw).Result
   )
   n1 = &RigVMFunction_MathDoubleRemap(
    SourceMaximum=-45
    SourceMinimum=-20
    TargetMaximum=1
    TargetMinimum=0
    Value=&RigVMFunction_MathQuaternionToEuler(
     RotationOrder=ZYX
     Value=&RigUnit_GetRelativeTransformForItem(Child=(Type=Bone,Name="spine_05"), Parent=(Type=Bone,Name="pelvis"), bChildInitial=False, bParentInitial=False).RelativeTransform.Rotation
    ).Result.X
   )
   &RigUnit_ModifyTransforms(
    ExecutePin=()
    ItemToModify=((Item=(Type=Bone,Name="Spine_metal_corr_up"),Transform=(Rotation=(X=0,Y=0,Z=0.0436,W=0.999),Translation=(X=0,Y=0,Z=0),Scale3D=(X=1,Y=1,Z=1))))
    WeightMinimum=0
    Weight=&RigVMFunction_KalmanFloat(BufferSize=33, Value=@n1.Result) {template="KalmanFilter::Execute(in Value,in BufferSize,out Result)"}.Result
   )
   n2 = &RigVMFunction_MathDoubleRemap(
    SourceMaximum=5
    SourceMinimum=-5
    TargetMaximum=1
    TargetMinimum=0
    Value=&RigVMFunction_MathQuaternionToEuler(
     RotationOrder=ZYX
     Value=&RigUnit_GetRelativeTransformForItem(Child=(Type=Bone,Name="spine_05"), Parent=(Type=Bone,Name="pelvis"), bChildInitial=False, bParentInitial=False).RelativeTransform.Rotation
    ).Result.X
   )
   n3 = RigVMVariableNode() {variable=Melee}
   > &RigUnit_ModifyTransforms(
    ItemToModify=((Item=(Type=Bone,Name="Spine_metal_corr_up"),Transform=(Rotation=(X=0,Y=0,Z=0.0872,W=0.9962),Translation=(X=0,Y=0,Z=0),Scale3D=(X=1,Y=1,Z=1))))
    WeightMinimum=0
    Weight=&RigVMFunction_KalmanFloat(BufferSize=33, Value=&RigVMFunction_MathDoubleMul(A=@n2.Result, B=@n3.Value).Result) {template="KalmanFilter::Execute(in Value,in BufferSize,out Result)"}.Result
   )
   &RigUnit_ModifyTransforms(
    ExecutePin=()
    ItemToModify="((Item=(Type=Bone,Name=\"Spine_metal_corr_up\"),Transform=(Rotation=(X=0,Y=0,Z=0.0436,W=0.999),Translation=(X=0,Y=-0.5,Z=0),Scale3D=(X=1,Y=1,Z=1))),(Item=(Type=Bone,Name=\"Spine_metal_corr_dwn\"),Transform=(Rotation=(X=0,Y=0,Z=0.0436,W=0.999),Translation=(X=0, ...(truncated)"
    Weight=1
    WeightMinimum=0
   )
   > Get Delta Ref_1 = RigVMFunctionReferenceNode(Item=(Type=Bone,Name="pelvis"))
   > &RigUnit_ModifyTransforms(
    ItemToModify=((Item=(Type=Bone,Name="Spine_metal_corr_up"),Transform=(Rotation=(X=0,Y=0,Z=0.0872,W=0.9962),Translation=(X=0,Y=0,Z=0),Scale3D=(X=1,Y=1,Z=1))))
    WeightMinimum=0
    Weight=&RigVMFunction_MathDoubleRemap(SourceMaximum=80, SourceMinimum=50, TargetMaximum=1, TargetMinimum=0, Value=@"Get Delta Ref_1".Result.Rotation.Yaw).Result
   )
   > &RigUnit_ModifyTransforms(
    ItemToModify=((Item=(Type=Bone,Name="Spine_metal_corr_up"),Transform=(Rotation=(X=0,Y=0,Z=0.1132,W=0.9936),Translation=(X=0,Y=0,Z=0),Scale3D=(X=1,Y=1,Z=1))))
    WeightMinimum=0
    Weight=&RigVMFunction_KalmanFloat(BufferSize=16, Value=RigVMVariableNode() {variable=OpenParawithDelay}.Value) {template="KalmanFilter::Execute(in Value,in BufferSize,out Result)"}.Result
   )
   > &RigUnit_ModifyTransforms(
    ItemToModify=((Item=(Type=Bone,Name="Spine_metal_corr_up"),Transform=(Rotation=(X=0,Y=0,Z=0.0436,W=0.999),Translation=(X=0,Y=0,Z=0),Scale3D=(X=1,Y=1,Z=1))))
    WeightMinimum=0
    Weight=&RigVMFunction_MathDoubleRemap(SourceMaximum=2100, SourceMinimum=1000, TargetMaximum=1, TargetMinimum=0, Value=RigVMVariableNode() {variable=Get Velocity Z}.Value).Result
   )
   n4 = &RigUnit_GetRelativeTransformForItem(Child=(Type=Bone,Name="spine_04"), Parent=(Type=Bone,Name="spine_03"), bChildInitial=False, bParentInitial=False)
   > &RigUnit_ModifyTransforms(
    ItemToModify=((Item=(Type=Bone,Name="Spine_metal_corr_up"),Transform=(Rotation=(X=0,Y=0,Z=0.0872,W=0.9962),Translation=(X=0,Y=0,Z=0),Scale3D=(X=1,Y=1,Z=1))))
    WeightMinimum=0
    Weight=&RigVMFunction_MathDoubleRemap(
     SourceMaximum=-15
     SourceMinimum=-10
     TargetMaximum=1
     TargetMinimum=0
     Value=&RigVMFunction_MathQuaternionToEuler(RotationOrder=XYZ, Value=@n4.RelativeTransform.Rotation).Result.Z
    ).Result
   )
   &RigUnit_ParentConstraint(
    AdvancedSettings=(InterpolationType=Average,RotationOrderForFilter=XZY)
    Child=(Type=Bone,Name="Spine_metal_corr_dwn")
    ExecutePin=()
    Filter=(TranslationFilter=(bX=True,bY=True,bZ=True),RotationFilter=(bX=True,bY=false,bZ=false),ScaleFilter=(bX=True,bY=True,bZ=True))
    Parents=((Item=(Type=Bone,Name="Spine_metal_corr_up"),Weight=1))
    Weight=1
   )
   > &RigUnit_ParentConstraint(
    AdvancedSettings=(InterpolationType=Average,RotationOrderForFilter=XZY)
    Child=(Type=Bone,Name="Spine_metal_corr_up")
    Filter=(TranslationFilter=(bX=True,bY=True,bZ=True),RotationFilter=(bX=True,bY=false,bZ=false),ScaleFilter=(bX=True,bY=True,bZ=True))
    Parents=((Item=(Type=Bone,Name="Base_chick"),Weight=1))
    Weight=1
   )
   RigUnit_BeginExecution = &RigUnit_BeginExecution()
   FN_DeformRig_FNC = RigVMFunctionReferenceNode(
    DeformRig_ Curve="(deformRig_l_upperarm_up_alpha_72_#1=False,deformRig_r_upperarm_up_alpha_73_#2=true,deformRig_l_upperarm_fr_alpha_74_#3=False,deformRig_r_upperarm_fr_alpha_75_CB51B80 ...(truncated)"
    DeformRig_Bool="(upperarm_1_#4=True,elbow_3_#5=True,wrist_7_#6=True,groin_8_#7=True,glute_17_#8=True,knee_9_40C408324CA261 ...(truncated)"
    DeformRig_Debug=(upperarm_debug_18_#4=False,wrist_debug_20_#6=False,lat_debug_26_#9=False,groinGlute_debug_27_#7=False)
    DeformRig_Value="(upperarm_up_18_#10=(Rotation=(X=0.0,Y=-0.0,Z=0.0,W=1.0),Translation=(X=0.0,Y=0.0,Z=8.0),Scale3D=(X=0.70,Y=1.0,Z=1.0)),upperarm_fwd_fr_19_#11=(Rotation=(X=0.0,Y=-0.0,Z=0.0,W=1.0),Translation=(X=0.0, ...(truncated)"
    ExecuteContext=@RigUnit_BeginExecution.ExecutePin
   )
   > RigVMFunction_Sequence = RigVMAggregateNode()
   Get Delta Ref_2_3_2_1 = RigVMFunctionReferenceNode(Item=(Type=Bone,Name="foot_l"), ExecuteContext=@RigVMFunction_Sequence.B)
   > &RigUnit_ModifyTransforms(
    ItemToModify="((Item=(Type=Bone,Name=\"Snikers_corr_l\"),Transform=(Rotation=(X=0,Y=0,Z=0,W=1),Translation=(X=-4,Y=1,Z=0),Scale3D=(X=1,Y=1,Z=1))),(Item=(Type=Bone,Name=\"Base_lape_l\"),Transform=(Rotation=(X=0,Y=0.2588,Z=0,W=0.9659),Translation=(X=0,Y=0,Z=0),Scale3D=(X=1,Y= ...(truncated)"
    WeightMinimum=0
    Weight=&RigVMFunction_MathDoubleRemap(SourceMaximum=50, SourceMinimum=30, TargetMaximum=1, TargetMinimum=0, Value=@"Get Delta Ref_2_3_2_1".Result.Rotation.Yaw).Result
   )
   > Get Delta Ref_2_3_1_1_1 = RigVMFunctionReferenceNode(Item=(Type=Bone,Name="foot_r"))
   > n5 = &RigUnit_ModifyTransforms(
    ItemToModify="((Item=(Type=Bone,Name=\"Snikers_corr_r\"),Transform=(Rotation=(X=0,Y=0,Z=0,W=1),Translation=(X=-5,Y=-1,Z=0),Scale3D=(X=1,Y=1,Z=1))),(Item=(Type=Bone,Name=\"Base_lape_r\"),Transform=(Rotation=(X=0,Y=0.2588,Z=0,W=0.9659),Translation=(X=0,Y=0,Z=0),Scale3D=(X=1,Y ...(truncated)"
    WeightMinimum=0
    Weight=&RigVMFunction_MathDoubleRemap(SourceMaximum=60, SourceMinimum=25, TargetMaximum=1, TargetMinimum=0, Value=@"Get Delta Ref_2_3_1_1_1".Result.Rotation.Yaw).Result
   )
   n6 = &RigVMFunction_ControlFlowBranch(ExecuteContext=@n5.ExecutePin, Condition=RigVMVariableNode() {variable=IsPlayingEmote}.Value)
   n7 = &RigUnit_SphericalPoseReader(
    ActiveRegionScaleFactors=(PositiveWidth=1,NegativeWidth=1,PositiveHeight=1,NegativeHeight=1)
    ActiveRegionSize=0.04
    Debug=(bDrawDebug=true,bDraw2D=False,bDrawLocalAxes=False,DebugScale=25,DebugSegments=20,DebugThickness=0.25)
    DriverAxis=(X=-1,Y=0,Z=0)
    DriverItem=(Type=Bone,Name="lowerarm_r")
    FalloffRegionScaleFactors=(PositiveWidth=1,NegativeWidth=1,PositiveHeight=1,NegativeHeight=1)
    FalloffSize=0.2
    FlipHeightScaling=False
    FlipWidthScaling=False
    OptionalParentItem=(Type=Bone,Name="upperarm_r")
    RotationOffset=(X=0,Y=100,Z=5)
    ExecutePin=@n6.True
   )
   > &RigUnit_ModifyTransforms(
    ItemToModify="((Item=(Type=Bone,Name=\"MSTR_ring\"),Transform=(Rotation=(X=0,Y=0,Z=0.342,W=0.9397),Translation=(X=0,Y=0,Z=0),Scale3D=(X=1,Y=1,Z=1))),(Item=(Type=Bone,Name=\"Base_ring\"),Transform=(Rotation=(X=0,Y=0.4226,Z=0,W=0.9063),Translation=(X=0.3,Y=1,Z=0),Scale3D=(X=1 ...(truncated)"
    WeightMinimum=0
    Weight=&RigVMFunction_KalmanFloat(BufferSize=6, Value=@n7.OutputParam) {template="KalmanFilter::Execute(in Value,in BufferSize,out Result)"}.Result
   )
   n8 = &RigVMFunction_KalmanFloat(BufferSize=8, Value=RigVMVariableNode() {variable=OpenParawithDelay}.Value) {template="KalmanFilter::Execute(in Value,in BufferSize,out Result)"}
   &RigUnit_ModifyTransforms(
    ItemToModify=((Item=(Type=Bone,Name="Spine_metal_corr_dwn"),Transform=(Rotation=(X=0,Y=0,Z=0,W=1),Translation=(X=0,Y=-3,Z=0),Scale3D=(X=1,Y=1,Z=1))))
    WeightMinimum=0
    Weight=@n8.Result
    ExecutePin=@RigVMFunction_Sequence.E
   )
   > &RigUnit_ModifyTransforms(
    ItemToModify=((Item=(Type=Bone,Name="Spine_metal_corr_up"),Transform=(Rotation=(X=0,Y=0,Z=0,W=1),Translation=(X=0,Y=-3,Z=0),Scale3D=(X=1,Y=1,Z=1))))
    WeightMinimum=0
    Weight=@n8.Result
   )
   > &RigUnit_ModifyTransforms(
    ItemToModify=((Item=(Type=Bone,Name="Spine_metal_corr_dwn"),Transform=(Rotation=(X=-0.0436,Y=0,Z=0,W=0.999),Translation=(X=0,Y=-1,Z=2),Scale3D=(X=1,Y=1,Z=1))))
    Weight=1
    WeightMinimum=0
   )
   RigVMFunction_MathQuaternionToEuler = &RigVMFunction_MathQuaternionToEuler(
    RotationOrder=ZYX
    Value=&RigUnit_GetRelativeTransformForItem(Child=(Type=Bone,Name="foot_r"), Parent=(Type=Bone,Name="calf_r"), bChildInitial=False, bParentInitial=False).RelativeTransform.Rotation
   )
   &RigUnit_ModifyTransforms(
    ItemToModify=((Item=(Type=Bone,Name="Dyn_lape_r"),Transform=(Rotation=(X=0.0378,Y=-0.4995,Z=0.0218,W=0.8652),Translation=(X=0,Y=0,Z=0),Scale3D=(X=1,Y=1,Z=1))))
    WeightMinimum=0
    Weight=&RigVMFunction_MathDoubleRemap(SourceMaximum=-50, SourceMinimum=-10, TargetMaximum=1, TargetMinimum=0, bClamp=False, Value=@RigVMFunction_MathQuaternionToEuler.Result.Z).Result
    ExecutePin=@RigVMFunction_Sequence.D
   )
   n9 = &RigUnit_GetRelativeTransformForItem(Child=(Type=Bone,Name="foot_l"), Parent=(Type=Bone,Name="calf_l"), bChildInitial=False, bParentInitial=False)
   > &RigUnit_ModifyTransforms(
    ItemToModify=((Item=(Type=Bone,Name="Dyn_lape_l"),Transform=(Rotation=(X=-0.1227,Y=-0.3391,Z=-0.0446,W=0.9317),Translation=(X=0,Y=0,Z=0),Scale3D=(X=1,Y=1,Z=1))))
    WeightMinimum=0
    Weight=&RigVMFunction_MathDoubleRemap(
     SourceMaximum=-30
     SourceMinimum=-5
     TargetMaximum=1
     TargetMinimum=0
     bClamp=False
     Value=&RigVMFunction_MathQuaternionToEuler(RotationOrder=ZYX, Value=@n9.RelativeTransform.Rotation).Result.Z
    ).Result
   )
   &RigUnit_ModifyTransforms(
    ItemToModify="((Item=(Type=Bone,Name=\"Base_Grenade_belt\"),Transform=(Rotation=(X=0,Y=0,Z=0,W=1),Translation=(X=-4,Y=1,Z=0),Scale3D=(X=1,Y=1,Z=1))),(Item=(Type=Bone,Name=\"Grenade_A\"),Transform=(Rotation=(X=0,Y=-0.0872,Z=0,W=0.9962),Translation=(X=-1.5,Y=0,Z=0),Scale3D=(X ...(truncated)"
    WeightMinimum=0
    Weight=RigVMDispatchNode(BoneName=root, Default=0, Name=deformRig_l_thigh_up_alpha) {template="DISPATCH_RigDispatch_GetAnimAttribute(in Name,in BoneName,in Default,out Value,out Found)"}.Value
    ExecutePin=@RigVMFunction_Sequence.C
   )
   > n10 = &RigUnit_ModifyTransforms(
    ItemToModify=((Item=(Type=Bone,Name="Base_Grenade_belt"),Transform=(Rotation=(X=0,Y=0,Z=0,W=1),Translation=(X=0,Y=0,Z=0),Scale3D=(X=1,Y=1,Z=1))))
    WeightMinimum=0
    Weight=RigVMVariableNode() {variable=BaseBallSliding}.Value
   )
   n11 = &RigVMFunction_ControlFlowBranch(
    Condition=/Script/RigVM.RigVMFunction_MathDoubleGreater(B=0, A=RigVMVariableNode() {variable=Two Arms Hang}.Value) {is=struct RigVMUnitNode, template="Greater::Execute(in A,in B,out Result)"}.Result
    ExecuteContext=@n10.ExecutePin
   )
   KalmanFilter = &RigVMFunction_KalmanFloat(BufferSize=16, Value=RigVMVariableNode() {variable=Two Arms Hang}.Value) {template="KalmanFilter::Execute(in Value,in BufferSize,out Result)"}
   &RigUnit_ModifyTransforms(
    ItemToModify="((Item=(Type=Bone,Name=\"Base_chick\"),Transform=(Rotation=(X=0,Y=-0.2588,Z=0,W=0.9659),Translation=(X=2,Y=0,Z=-8),Scale3D=(X=1,Y=1,Z=1))),(Item=(Type=Bone,Name=\"Base_chick__A\"),Transform=(Rotation=(X=0,Y=0,Z=0,W=1),Translation=(X=0,Y=0,Z=-3),Scale3D=(X=1,Y= ...(truncated)"
    WeightMinimum=0
    Weight=@KalmanFilter.Result
    ExecutePin=@n11.True
   )
   > &RigUnit_ModifyTransforms(
    ItemToModify="((Item=(Type=Bone,Name=\"deform_upperarm_up_l\"),Transform=(Rotation=(X=0,Y=0,Z=0,W=1),Translation=(X=10,Y=0,Z=-10),Scale3D=(X=1,Y=1,Z=1))),(Item=(Type=Bone,Name=\"deform_upperarm_up_r\"),Transform=(Rotation=(X=0,Y=0,Z=0,W=1),Translation=(X=0,Y=0,Z=10),Scale3D ...(truncated)"
    WeightMinimum=0
    Weight=@KalmanFilter.Result
   )
   &RigUnit_ParentConstraint(
    AdvancedSettings=(InterpolationType=Average,RotationOrderForFilter=XZY)
    Child=(Type=Bone,Name="Base_chick")
    Parents=((Item=(Type=Bone,Name="clavicle_r"),Weight=0.5),(Item=(Type=Bone,Name="spine_05"),Weight=0.5))
    Weight=1
    ExecutePin=@n11.False
   )
   > n12 = &RigUnit_SphericalPoseReader(
    ActiveRegionScaleFactors=(PositiveWidth=1.5,NegativeWidth=1,PositiveHeight=1,NegativeHeight=2.3)
    ActiveRegionSize=0.1
    Debug=(bDrawDebug=false,bDraw2D=False,bDrawLocalAxes=False,DebugScale=25,DebugSegments=20,DebugThickness=0.25)
    DriverAxis=(X=-1,Y=0,Z=0)
    DriverItem=(Type=Bone,Name="upperarm_r")
    FalloffRegionScaleFactors=(PositiveWidth=1,NegativeWidth=0.5,PositiveHeight=1.6,NegativeHeight=2)
    FalloffSize=0.3
    FlipHeightScaling=False
    FlipWidthScaling=False
    OptionalParentItem=(Type=Bone,Name="spine_05")
    RotationOffset=(X=-100,Y=30,Z=30)
   )
   > &RigUnit_ModifyTransforms(
    ItemToModify="((Item=(Type=Bone,Name=\"Base_chick\"),Transform=(Rotation=(X=0.1736,Y=0,Z=0,W=0.9848),Translation=(X=0,Y=-4,Z=1),Scale3D=(X=1,Y=1,Z=1))),(Item=(Type=Bone,Name=\"MSTR_ring\"),Transform=(Rotation=(X=0,Y=0,Z=0.1908,W=0.9816),Translation=(X=0,Y=0,Z=0),Scale3D=(X= ...(truncated)"
    WeightMinimum=0
    Weight=@n12.OutputParam
   )
   Remap = &RigVMFunction_MathDoubleRemap(
    SourceMaximum=45
    SourceMinimum=55
    TargetMaximum=1
    TargetMinimum=0
    Value=&RigUnit_GetRelativeTransformForItem(Child=(Type=Bone,Name="pelvis"), Parent=(Type=Bone,Name="root"), bChildInitial=False, bParentInitial=False).RelativeTransform.Translation.Z
   )
   RigVMVariableNode(Value=@Remap.Result, ExecuteContext=@RigVMFunction_Sequence.A) {variable=SIt}
   Multiply = &RigVMFunction_MathDoubleMul(
    B=&RigVMFunction_MathDoubleRemap(
     SourceMaximum=60
     SourceMinimum=50
     TargetMaximum=1
     TargetMinimum=0
     Value=&RigUnit_GetRelativeTransformForItem(Child=(Type=Bone,Name="pelvis"), Parent=(Type=Bone,Name="root"), bChildInitial=False, bParentInitial=False).RelativeTransform.Translation.Z
    ).Result
    A=&RigVMFunction_MathDoubleRemap(
     SourceMaximum=77
     SourceMinimum=85
     TargetMaximum=1
     TargetMinimum=0
     Value=&RigUnit_GetRelativeTransformForItem(Child=(Type=Bone,Name="pelvis"), Parent=(Type=Bone,Name="root"), bChildInitial=False, bParentInitial=False).RelativeTransform.Translation.Z
    ).Result
   )
   > n13 = RigVMVariableNode(Value=@Multiply.Result) {variable=Crouch}
   RigVMFunction_ControlFlowBranch = &RigVMFunction_ControlFlowBranch(Condition=RigVMVariableNode() {variable=GauntletsEquipt}.Value, ExecuteContext=@n13.ExecuteContext)
   IK correction BitterSweet = RigVMFunctionReferenceNode(Disable Arms=1, ExecuteContext=@RigVMFunction_ControlFlowBranch.False)
   > &RigUnit_ModifyTransforms(
    ItemToModify=((Item=(Type=Bone,Name="Spine_metal_corr_up"),Transform=(Rotation=(X=0,Y=0,Z=0,W=1),Translation=(X=0,Y=-4,Z=0),Scale3D=(X=1,Y=1,Z=1))))
    Weight=1
    WeightMinimum=0
   )
   n14 = &RigUnit_SphericalPoseReader(
    ActiveRegionScaleFactors=(PositiveWidth=1.6,NegativeWidth=1,PositiveHeight=1,NegativeHeight=1)
    ActiveRegionSize=0.1
    Debug=(bDrawDebug=true,bDraw2D=False,bDrawLocalAxes=False,DebugScale=25,DebugSegments=20,DebugThickness=0.25)
    DriverAxis=(X=1,Y=0,Z=0)
    DriverItem=(Type=Bone,Name="lowerarm_l")
    ExecutePin=()
    FalloffRegionScaleFactors=(PositiveWidth=0.7,NegativeWidth=0.7,PositiveHeight=2,NegativeHeight=1)
    FalloffSize=0.4
    FlipHeightScaling=False
    FlipWidthScaling=False
    OptionalParentItem=(Type=Bone,Name="spine_05")
    RotationOffset=(X=140,Y=1,Z=0)
   )
   > &RigUnit_ModifyTransforms(
    ItemToModify="((Item=(Type=Bone,Name=\"Spine_metal_corr_up\"),Transform=(Rotation=(X=0,Y=0,Z=0,W=1),Translation=(X=5,Y=0,Z=0),Scale3D=(X=1,Y=1,Z=1))),(Item=(Type=Bone,Name=\"Spine_metal_corr_dwn\"),Transform=(Rotation=(X=0,Y=0.0872,Z=0,W=0.9962),Translation=(X=0,Y=0,Z=0),Sc ...(truncated)"
    WeightMinimum=0
    Weight=@n14.OutputParam
   )
   n15 = &RigUnit_SphericalPoseReader(
    ActiveRegionScaleFactors=(PositiveWidth=1.61,NegativeWidth=1,PositiveHeight=1,NegativeHeight=1)
    ActiveRegionSize=0.1
    Debug=(bDrawDebug=true,bDraw2D=False,bDrawLocalAxes=False,DebugScale=25,DebugSegments=20,DebugThickness=0.25)
    DriverAxis=(X=1,Y=0,Z=0)
    DriverItem=(Type=Bone,Name="lowerarm_l")
    ExecutePin=()
    FalloffRegionScaleFactors=(PositiveWidth=2,NegativeWidth=1,PositiveHeight=2,NegativeHeight=2)
    FalloffSize=0.2
    FlipHeightScaling=False
    FlipWidthScaling=False
    OptionalParentItem=(Type=Bone,Name="spine_05")
    RotationOffset=(X=50,Y=50,Z=30)
   )
   > &RigUnit_ModifyTransforms(
    ItemToModify=((Item=(Type=Bone,Name="Spine_metal_corr_up"),Transform=(Rotation=(X=0,Y=0,Z=-0.0262,W=0.9997),Translation=(X=5,Y=0,Z=0),Scale3D=(X=1,Y=1,Z=1))))
    WeightMinimum=0
    Weight=@n15.OutputParam
   )
   n16 = &RigUnit_GetRelativeTransformForItem(Child=(Type=Bone,Name="spine_05"), Parent=(Type=Bone,Name="pelvis"), bChildInitial=False, bParentInitial=False)
   &RigUnit_ModifyTransforms(
    ExecutePin=()
    ItemToModify="((Item=(Type=Bone,Name=\"Spine_metal_corr_up\"),Transform=(Rotation=(X=0,Y=0,Z=0.0872,W=0.9962),Translation=(X=0,Y=0,Z=0),Scale3D=(X=1,Y=1,Z=1))),(Item=(Type=Bone,Name=\"Spine_metal_corr_dwn\"),Transform=(Rotation=(X=0,Y=0,Z=0.0436,W=0.999),Translation=(X=0,Y= ...(truncated)"
    WeightMinimum=0
    Weight=&RigVMFunction_MathDoubleRemap(
     SourceMaximum=-20
     SourceMinimum=-15
     TargetMaximum=1
     TargetMinimum=0
     Value=&RigVMFunction_MathQuaternionToEuler(RotationOrder=ZYX, Value=@n16.RelativeTransform.Rotation).Result.Z
    ).Result
   )
   &RigUnit_ModifyTransforms(
    ExecutePin=()
    ItemToModify="((Item=(Type=Bone,Name=\"Spine_metal_corr_up\"),Transform=(Rotation=(X=0,Y=0,Z=0,W=1),Translation=(X=-5,Y=0,Z=0),Scale3D=(X=1,Y=1,Z=1))),(Item=(Type=Bone,Name=\"Spine_metal_corr_dwn\"),Transform=(Rotation=(X=0,Y=-0.0872,Z=0,W=0.9962),Translation=(X=0,Y=0,Z=0), ...(truncated)"
    Weight=1
    WeightMinimum=0
   )
   &RigVMFunction_MathDoubleMul(A=@n1.Result, B=RigVMVariableNode() {variable=Melee}.Value)
   &RigVMFunction_MathDoubleMul(A=1, B=1)
   &RigUnit_ParentConstraint(
    AdvancedSettings=(InterpolationType=Average,RotationOrderForFilter=XZY)
    Child=(Type=Bone,Name="Spine_metal_corr_up")
    ExecutePin=()
    Parents=((Item=(Type=Bone,Name="Base_chick"),Weight=0.5),(Item=(Type=Bone,Name="spine_03"),Weight=0.5))
    Weight=1
   )
   &RigUnit_SphericalPoseReader(
    ActiveRegionScaleFactors=(PositiveWidth=1.6,NegativeWidth=1,PositiveHeight=1,NegativeHeight=1)
    ActiveRegionSize=0.1
    Debug=(bDrawDebug=true,bDraw2D=False,bDrawLocalAxes=False,DebugScale=25,DebugSegments=20,DebugThickness=0.25)
    DriverAxis=(X=1,Y=0,Z=0)
    DriverItem=(Type=Bone,Name="upperarm_l")
    ExecutePin=()
    FalloffRegionScaleFactors=(PositiveWidth=0.7,NegativeWidth=0.7,PositiveHeight=1.3,NegativeHeight=1)
    FalloffSize=0.4
    FlipHeightScaling=False
    FlipWidthScaling=False
    OptionalParentItem=(Type=Bone,Name="spine_05")
    RotationOffset=(X=27.9774,Y=10,Z=-24.4969)
   )
  ]}
]
hierarchy [
 root {loc=<0 0 0>, rot=<0 0 0 1>}
 pelvis {parent=root, loc=<0 -0.7321 92.4218>, rot=<0.0127 0.707 -0.0127 -0.707>}
 spine_01 {parent=pelvis, loc=<5.6002 -0.5128 0>, rot=<0 0 0.034 -0.9994>}
 spine_02 {parent=spine_01, loc=<8.0756 0.0139 0>, rot=<0 0 0.0238 -0.9997>}
 spine_03 {parent=spine_02, loc=<7.0395 0 0>, rot=<0 0 0.0149 -0.9999>}
 spine_04 {parent=spine_03, loc=<8.1023 0 0>, rot=<0 0 -0.0132 -0.9999>}
 spine_05 {parent=spine_04, loc=<10.0141 0 0>, rot=<0 0 -0.0015 -1>}
 clavicle_l {parent=spine_05, loc=<2.5856 -0.2866 -4.3777>, rot=<0.0465 -0.7289 -0.009 -0.6829>}
 upperarm_l {parent=clavicle_l, loc=<18.2264 0.1342 -0.001>, rot=<0.0253 -0.2693 0.0013 -0.9627>}
 lowerarm_l {parent=upperarm_l, loc=<29.0071 0.1686 -1.239>, rot=<0.0391 -0.0099 -0.2453 -0.9686>}
 hand_l {parent=lowerarm_l, loc=<25.8607 0.222 1.0274>, rot=<-0.7193 -0.0988 0.1193 -0.6773>}
 thumb_01_l {parent=hand_l, loc=<1.3118 -0.3191 -2.459>, rot=<0.6546 -0.4042 0.0664 -0.6354>}
 thumb_02_l {parent=thumb_01_l, loc=<5.9336 -0.0448 0.0401>, rot=<-0.0037 0.0213 0.298 -0.9543>}
 thumb_03_l {parent=thumb_02_l, loc=<2.2223 0.3682 -0.0907>, rot=<0.0118 -0.013 0.0066 -0.9998>}
 index_metacarpal_l {parent=hand_l, loc=<1.6535 0.0183 -2.96>, rot=<0.0158 -0.0124 -0.0173 -0.9996>}
 index_01_l {parent=index_metacarpal_l, loc=<8.3008 0.0357 -0.0524>, rot=<-0.0044 0.0485 0.0924 -0.9945>}
 index_02_l {parent=index_01_l, loc=<3.3587 -0.3004 0.0452>, rot=<-0.002 0.0027 0.3225 -0.9466>}
 index_03_l {parent=index_02_l, loc=<2.5578 -0.123 -0.0252>, rot=<-0.002 0.0067 0.2558 -0.9667>}
 middle_metacarpal_l {parent=hand_l, loc=<2.1856 -0.1725 -0.5984>, rot=<0.006 0.0168 -0.0329 -0.9993>}
 middle_01_l {parent=middle_metacarpal_l, loc=<7.3275 -0.2447 0.0129>, rot=<-0.0208 0.0042 0.0858 -0.9961>}
 middle_02_l {parent=middle_01_l, loc=<4.0123 0.1401 0.1175>, rot=<-0.0218 0.0258 0.3512 -0.9357>}
 middle_03_l {parent=middle_02_l, loc=<2.6077 0.0534 0.1373>, rot=<0.0082 0.0188 0.1708 -0.9851>}
 ring_metacarpal_l {parent=hand_l, loc=<2.2074 -0.3625 1.3638>, rot=<-0.0284 0.0722 -0.0458 -0.9959>}
 ring_01_l {parent=ring_metacarpal_l, loc=<7.0295 -0.6197 -0.1446>, rot=<-0.0489 -0.0252 0.1892 -0.9804>}
 ring_02_l {parent=ring_01_l, loc=<3.6174 0.4647 -0.1027>, rot=<0 0 0.2806 -0.9598>}
 ring_03_l {parent=ring_02_l, loc=<2.4152 0.114 0.0199>, rot=<0 0 0.1189 -0.9929>}
 pinky_metacarpal_l {parent=hand_l, loc=<1.7616 -0.9753 3.3814>, rot=<-0.0397 0.0814 -0.069 -0.9935>}
 pinky_01_l {parent=pinky_metacarpal_l, loc=<7.1709 -0.4831 -0.1094>, rot=<-0.0654 -0.0035 0.2242 -0.9723>}
 pinky_02_l {parent=pinky_01_l, loc=<2.8274 0.0893 -0.1043>, rot=<0.0037 0.0011 0.2854 -0.9584>}
 pinky_03_l {parent=pinky_02_l, loc=<2.2046 0.0307 -0.0078>, rot=<-0.0002 0.0098 0.1445 -0.9895>}
 deform_wrist_dn_base_l {parent=hand_l, loc=<0 -3.4173 0>, rot=<0 0 0 -1>}
 deform_wrist_dn_l {parent=deform_wrist_dn_base_l, loc=<0 0 0>, rot=<0 0 0 1>}
 deform_wrist_up_base_l {parent=hand_l, loc=<0 2.7805 0>, rot=<0 0 0 -1>}
 deform_wrist_up_l {parent=deform_wrist_up_base_l, loc=<0 0 0>, rot=<0 0 0 1>}
 lowerarm_twist_02_l {parent=lowerarm_l, loc=<9.6199 0 0>, rot=<0 0 0 -1>}
 lowerarm_twist_01_l {parent=lowerarm_l, loc=<17.8549 0 0>, rot=<0 0 0 -1>}
 deform_elbow_in_base_l {parent=lowerarm_l, loc=<0.2112 5.645 0.8274>, rot=<0 0 0 -1>}
 deform_elbow_in_l {parent=deform_elbow_in_base_l, loc=<0 0 0>, rot=<0 0 0 1>}
 deform_elbow_out_base_l {parent=lowerarm_l, loc=<0 -7.5129 0>, rot=<0 0 0 -1>}
 deform_elbow_out_l {parent=deform_elbow_out_base_l, loc=<0 0 0>, rot=<0 0 0 1>}
 upperarm_twist_01_l {parent=upperarm_l, loc=<7.9796 0.0843 -0.6195>, rot=<0 0 0 -1>}
 deform_upperarm_bk_base_l {parent=upperarm_twist_01_l, loc=<-6.8983 -13.3826 -1.3258>, rot=<0 0 0 1>}
 deform_upperarm_bk_l {parent=deform_upperarm_bk_base_l, loc=<0 0 0>, rot=<0 0 0 1>}
 deform_upperarm_fr_base_l {parent=upperarm_twist_01_l, loc=<-8.3487 10.7117 -2.1455>, rot=<0 0 0 1>}
 deform_upperarm_fr_l {parent=deform_upperarm_fr_base_l, loc=<0 0 0>, rot=<0 0 0 1>}
 deform_upperarm_up_base_l {parent=upperarm_twist_01_l, loc=<-11.2012 -1.2234 8.7784>, rot=<0 0 0 1>}
 deform_upperarm_up_l {parent=deform_upperarm_up_base_l, loc=<0 0 0>, rot=<0 0 0 1>}
 upperarm_twist_02_l {parent=upperarm_l, loc=<19.5059 0.0843 -0.6195>, rot=<0 0 0 -1>}
 deform_bicep_base_l {parent=upperarm_twist_02_l, loc=<0 8.0494 0>, rot=<0 0 0 1>}
 deform_bicep_l {parent=deform_bicep_base_l, loc=<0 0 0>, rot=<0 0 0 1>}
 p_deform_upperarm_l {parent=clavicle_l, loc=<18.2264 0.1342 -0.001>, rot=<0.0253 -0.2693 0.0013 -0.9627>}
 p_deform_lowerarm_l {parent=p_deform_upperarm_l, loc=<29.0071 0.1686 -1.239>, rot=<0.0391 -0.0099 -0.2453 -0.9686>}
 p_deform_hand_l {parent=p_deform_lowerarm_l, loc=<25.8607 0.222 1.0274>, rot=<-0.7193 -0.0988 0.1193 -0.6773>}
 p_deform_lowerarm_twist_02_l {parent=p_deform_lowerarm_l, loc=<9.6199 0 0>, rot=<0 0 0 -1>}
 p_deform_lowerarm_twist_01_l {parent=p_deform_lowerarm_l, loc=<17.8549 0 0>, rot=<0 0 0 -1>}
 p_deform_elbow_in_base_l {parent=p_deform_lowerarm_l, loc=<0.2112 5.645 0.8274>, rot=<0 0 0 -1>}
 p_deform_elbow_in_l {parent=p_deform_elbow_in_base_l, loc=<0 0 0>, rot=<0 0 0 1>}
 p_deform_elbow_out_base_l {parent=p_deform_lowerarm_l, loc=<0 -7.5129 0>, rot=<0 0 0 -1>}
 p_deform_elbow_out_l {parent=p_deform_elbow_out_base_l, loc=<0 0 0>, rot=<0 0 0 1>}
 p_deform_upperarm_twist_01_l {parent=p_deform_upperarm_l, loc=<7.9796 0.0843 -0.6195>, rot=<0 0 0 -1>}
 p_deform_upperarm_up_base_l {parent=p_deform_upperarm_twist_01_l, loc=<-11.2012 -1.2234 8.7784>, rot=<0 0 0 -1>}
 p_deform_upperarm_up_l {parent=p_deform_upperarm_up_base_l, loc=<0 0 0>, rot=<0 0 0 1>}
 p_deform_upperarm_fr_base_l {parent=p_deform_upperarm_twist_01_l, loc=<-8.3487 10.7117 -2.1455>, rot=<0 0 0 -1>}
 p_deform_upperarm_fr_l {parent=p_deform_upperarm_fr_base_l, loc=<0 0 0>, rot=<0 0 0 1>}
 p_deform_upperarm_bk_base_l {parent=p_deform_upperarm_twist_01_l, loc=<-6.8983 -13.3826 -1.3258>, rot=<0 0 0 -1>}
 p_deform_upperarm_bk_l {parent=p_deform_upperarm_bk_base_l, loc=<0 0 0>, rot=<0 0 0 1>}
 p_deform_upperarm_twist_02_l {parent=p_deform_upperarm_l, loc=<19.5059 0.0843 -0.6195>, rot=<0 0 0 -1>}
 p_deform_bicep_base_l {parent=p_deform_upperarm_twist_02_l, loc=<0 8.0494 0>, rot=<0 0 0 1>}
 p_deform_bicep_l {parent=p_deform_bicep_base_l, loc=<0 0 0>, rot=<0 0 0 1>}
 neck_01 {parent=spine_05, loc=<10.815 1.9196 0>, rot=<0 0 -0.1864 -0.9825>}
 neck_02 {parent=neck_01, loc=<4.7347 0.3045 0>, rot=<0 0 0.0363 -0.9993>}
 head {parent=neck_02, loc=<5.7632 0.5552 0>, rot=<0 0 0.115 -0.9934>}
 Base_zipper {parent=neck_01, loc=<0.9372 18.9266 -0.0137>, rot=<-0.2331 -0.6536 0.6811 0.2337>}
 Dyn_zipper {parent=Base_zipper, loc=<0.9887 0.0122 -0.0389>, rot=<0 -0.0459 0 -0.9989>}
 clavicle_r {parent=spine_05, loc=<2.5855 -0.2866 4.3777>, rot=<0.7289 0.0465 -0.6829 0.009>}
 upperarm_r {parent=clavicle_r, loc=<-18.2264 -0.1342 0.001>, rot=<0.0253 -0.2693 0.0013 -0.9627>}
 lowerarm_r {parent=upperarm_r, loc=<-29.0071 -0.1686 1.239>, rot=<0.0391 -0.0099 -0.2453 -0.9686>}
 hand_r {parent=lowerarm_r, loc=<-25.8607 -0.222 -1.0274>, rot=<-0.7193 -0.0988 0.1193 -0.6773>}
 thumb_01_r {parent=hand_r, loc=<-1.3118 0.3191 2.459>, rot=<0.6546 -0.4042 0.0664 -0.6354>}
 thumb_02_r {parent=thumb_01_r, loc=<-5.9336 0.0448 -0.0401>, rot=<-0.0037 0.0213 0.298 -0.9543>}
 thumb_03_r {parent=thumb_02_r, loc=<-2.2223 -0.3682 0.0907>, rot=<0.0118 -0.013 0.0066 -0.9998>}
 index_metacarpal_r {parent=hand_r, loc=<-1.6535 -0.0183 2.96>, rot=<0.0158 -0.0124 -0.0173 -0.9996>}
 index_01_r {parent=index_metacarpal_r, loc=<-8.3008 -0.0357 0.0524>, rot=<-0.0044 0.0485 0.0924 -0.9945>}
 index_02_r {parent=index_01_r, loc=<-3.3587 0.3004 -0.0452>, rot=<-0.002 0.0027 0.3225 -0.9466>}
 index_03_r {parent=index_02_r, loc=<-2.5578 0.123 0.0252>, rot=<-0.002 0.0067 0.2558 -0.9667>}
 middle_metacarpal_r {parent=hand_r, loc=<-2.1856 0.1725 0.5984>, rot=<0.006 0.0168 -0.0329 -0.9993>}
 middle_01_r {parent=middle_metacarpal_r, loc=<-7.3275 0.2447 -0.0129>, rot=<-0.0208 0.0042 0.0858 -0.9961>}
 middle_02_r {parent=middle_01_r, loc=<-4.0123 -0.1401 -0.1175>, rot=<-0.0218 0.0258 0.3512 -0.9357>}
 middle_03_r {parent=middle_02_r, loc=<-2.6077 -0.0534 -0.1373>, rot=<0.0082 0.0188 0.1708 -0.9851>}
 ring_metacarpal_r {parent=hand_r, loc=<-2.2074 0.3625 -1.3638>, rot=<-0.0284 0.0722 -0.0458 -0.9959>}
 ring_01_r {parent=ring_metacarpal_r, loc=<-7.0295 0.6197 0.1446>, rot=<-0.0489 -0.0252 0.1892 -0.9804>}
 ring_02_r {parent=ring_01_r, loc=<-3.6174 -0.4647 0.1027>, rot=<0 0 0.2806 -0.9598>}
 ring_03_r {parent=ring_02_r, loc=<-2.4152 -0.114 -0.0199>, rot=<0 0 0.1189 -0.9929>}
 pinky_metacarpal_r {parent=hand_r, loc=<-1.7616 0.9753 -3.3814>, rot=<-0.0397 0.0814 -0.069 -0.9935>}
 pinky_01_r {parent=pinky_metacarpal_r, loc=<-7.1709 0.4831 0.1094>, rot=<-0.0654 -0.0035 0.2242 -0.9723>}
 pinky_02_r {parent=pinky_01_r, loc=<-2.8274 -0.0893 0.1043>, rot=<0.0037 0.0011 0.2854 -0.9584>}
 pinky_03_r {parent=pinky_02_r, loc=<-2.2046 -0.0307 0.0078>, rot=<-0.0002 0.0098 0.1445 -0.9895>}
 deform_wrist_dn_base_r {parent=hand_r, loc=<0 3.4172 0>, rot=<0 0 0 -1>}
 deform_wrist_dn_r {parent=deform_wrist_dn_base_r, loc=<0 0 0>, rot=<0 0 0 -1>}
 deform_wrist_up_base_r {parent=hand_r, loc=<-0.0001 -2.7804 0>, rot=<0 0 0 -1>}
 deform_wrist_up_r {parent=deform_wrist_up_base_r, loc=<0 0 0>, rot=<0 0 0 -1>}
 lowerarm_twist_02_r {parent=lowerarm_r, loc=<-9.6199 0 0>, rot=<0 0 0 -1>}
 lowerarm_twist_01_r {parent=lowerarm_r, loc=<-17.8549 0 0>, rot=<0 0 0 -1>}
 deform_elbow_in_base_r {parent=lowerarm_r, loc=<-0.2114 -5.6449 -0.827>, rot=<0 0 0 -1>}
 deform_elbow_in_r {parent=deform_elbow_in_base_r, loc=<0 0 0>, rot=<0 0 0 1>}
 deform_elbow_out_base_r {parent=lowerarm_r, loc=<-0.0001 7.5129 0.0002>, rot=<0 0 0 -1>}
 deform_elbow_out_r {parent=deform_elbow_out_base_r, loc=<0 0 0>, rot=<0 0 0 1>}
 upperarm_twist_01_r {parent=upperarm_r, loc=<-7.9796 -0.0843 0.6195>, rot=<0 0 0 -1>}
 deform_upperarm_bk_base_r {parent=upperarm_twist_01_r, loc=<6.8985 13.3826 1.3257>, rot=<0 0 0 1>}
 deform_upperarm_bk_r {parent=deform_upperarm_bk_base_r, loc=<0 0 0>, rot=<0 0 0 1>}
 deform_upperarm_fr_base_r {parent=upperarm_twist_01_r, loc=<8.3485 -10.7117 2.1456>, rot=<0 0 0 1>}
 deform_upperarm_fr_r {parent=deform_upperarm_fr_base_r, loc=<0 0 0>, rot=<0 0 0 1>}
 deform_upperarm_up_base_r {parent=upperarm_twist_01_r, loc=<11.2015 1.2234 -8.7788>, rot=<0 0 0 1>}
 deform_upperarm_up_r {parent=deform_upperarm_up_base_r, loc=<0 0 0>, rot=<0 0 0 1>}
 upperarm_twist_02_r {parent=upperarm_r, loc=<-19.5059 -0.0843 0.6195>, rot=<0 0 0 -1>}
 deform_bicep_base_r {parent=upperarm_twist_02_r, loc=<0.0002 -8.0494 -0.0003>, rot=<0 0 0 1>}
 deform_bicep_r {parent=deform_bicep_base_r, loc=<0 0 0>, rot=<0 0 0 1>}
 Base_chick {parent=clavicle_r, loc=<-17.4587 2.4501 -15.005>, rot=<0.8435 -0.0346 0.5312 0.0722>}
 Base_chick__B {parent=Base_chick, loc=<0 0 -4.7074>, rot=<0 0 0 1>}
 Base_chick__A {parent=Base_chick, loc=<0 0 4.5749>, rot=<0 0 0 1>}
 MSTR_ring {parent=Base_chick, loc=<0.1332 -2.5555 -6.4137>, rot=<0 0 0 1>}
 Base_ring {parent=MSTR_ring, loc=<0 0 0>, rot=<0 0 0 1>}
 Dyn_ring {parent=Base_ring, loc=<0.2282 0.3579 -1.7459>, rot=<0 0 0 1>}
 Base_ribbon_FR {parent=clavicle_r, loc=<-15.2041 -15.2882 -4.0633>, rot=<0.7665 -0.4062 0.4778 -0.1385>}
 Dyn_ribbon_FR {parent=Base_ribbon_FR, loc=<0 0 0>, rot=<0 0 0 1>}
 Base_ribbon_BK {parent=clavicle_r, loc=<-13.667 12.1098 -11.1328>, rot=<0.8364 0.1321 0.4462 0.2897>}
 Dyn_ribbon_BK {parent=Base_ribbon_BK, loc=<0 0 0>, rot=<0 0 0 1>}
 p_deform_upperarm_r {parent=clavicle_r, loc=<-18.2264 -0.1342 0.001>, rot=<0.0253 -0.2693 0.0013 -0.9627>}
 p_deform_lowerarm_r {parent=p_deform_upperarm_r, loc=<-29.0071 -0.1686 1.239>, rot=<0.0391 -0.0099 -0.2453 -0.9686>}
 p_deform_hand_r {parent=p_deform_lowerarm_r, loc=<-25.8607 -0.222 -1.0274>, rot=<-0.7193 -0.0988 0.1193 -0.6773>}
 p_deform_lowerarm_twist_02_r {parent=p_deform_lowerarm_r, loc=<-9.6199 0 0>, rot=<0 0 0 -1>}
 p_deform_lowerarm_twist_01_r {parent=p_deform_lowerarm_r, loc=<-17.8549 0 0>, rot=<0 0 0 -1>}
 p_deform_elbow_out_base_r {parent=p_deform_lowerarm_r, loc=<-0.0001 7.5129 0.0002>, rot=<0 0 0 -1>}
 p_deform_elbow_out_r {parent=p_deform_elbow_out_base_r, loc=<0 0 0>, rot=<0 0 0 1>}
 p_deform_elbow_in_base_r {parent=p_deform_lowerarm_r, loc=<-0.2114 -5.6449 -0.827>, rot=<0 0 0 -1>}
 p_deform_elbow_in_r {parent=p_deform_elbow_in_base_r, loc=<0 0 0>, rot=<0 0 0 1>}
 p_deform_upperarm_twist_01_r {parent=p_deform_upperarm_r, loc=<-7.9796 -0.0843 0.6195>, rot=<0 0 0 -1>}
 p_deform_upperarm_up_base_r {parent=p_deform_upperarm_twist_01_r, loc=<11.2015 1.2234 -8.7788>, rot=<0 0 0 -1>}
 p_deform_upperarm_up_r {parent=p_deform_upperarm_up_base_r, loc=<0 0 0>, rot=<0 0 0 1>}
 p_deform_upperarm_fr_base_r {parent=p_deform_upperarm_twist_01_r, loc=<8.3485 -10.7117 2.1456>, rot=<0 0 0 -1>}
 p_deform_upperarm_fr_r {parent=p_deform_upperarm_fr_base_r, loc=<0 0 0>, rot=<0 0 0 1>}
 p_deform_upperarm_bk_base_r {parent=p_deform_upperarm_twist_01_r, loc=<6.8985 13.3826 1.3257>, rot=<0 0 0 -1>}
 p_deform_upperarm_bk_r {parent=p_deform_upperarm_bk_base_r, loc=<0 0 0>, rot=<0 0 0 1>}
 p_deform_upperarm_twist_02_r {parent=p_deform_upperarm_r, loc=<-19.5059 -0.0843 0.6195>, rot=<0 0 0 -1>}
 p_deform_bicep_base_r {parent=p_deform_upperarm_twist_02_r, loc=<0.0002 -8.0494 -0.0003>, rot=<0 0 0 1>}
 p_deform_bicep_r {parent=p_deform_bicep_base_r, loc=<0 0 0>, rot=<0 0 0 1>}
 attach_cape {parent=spine_05, loc=<9.0169 -11.714 0>, rot=<0.0616 -0.7044 -0.0616 -0.7044>}
 cape {parent=attach_cape, type=Null, loc=<0 -9.48 7.3319>, rot=<0 0 0 1>}
 attach_backpack {parent=spine_05, loc=<-0.123 -13.473 0>, rot=<-0.0093 -0.707 0.0093 -0.707>}
 Backpack_BR {parent=attach_backpack, type=Null, loc=<0 -11.3023 0>, rot=<0 0 0 1>}
 deform_pec_aim_base_l {parent=spine_04, loc=<5.0227 13.6719 -2.6508>, rot=<0.0672 -0.5201 0.3228 -0.7879>}
 deform_pec_aim_l {parent=deform_pec_aim_base_l, loc=<0 0 0>, rot=<0 0 0 1>}
 deform_pec_base_l {parent=deform_pec_aim_l, loc=<16.3279 0 0>, rot=<0 0 0 1>}
 deform_pec_l {parent=deform_pec_base_l, loc=<0 0 0>, rot=<0 0 0 1>}
 deform_pec_aim_base_r {parent=spine_04, loc=<5.0225 13.6718 2.6508>, rot=<-0.5201 -0.0672 0.7879 0.3228>}
 deform_pec_aim_r {parent=deform_pec_aim_base_r, loc=<0 0 0>, rot=<0 0 0 1>}
 deform_pec_base_r {parent=deform_pec_aim_r, loc=<-16.3279 0 0>, rot=<0 0 0 -1>}
 deform_pec_r {parent=deform_pec_base_r, loc=<0 0 0.0002>, rot=<0 0 0 -1>}
 deform_lat_l_base {parent=spine_03, loc=<2.5561 -10.3716 -17.6918>, rot=<0 0 0 -1>}
 deform_lat_l {parent=deform_lat_l_base, loc=<0 0 0>, rot=<0 0 0 1>}
 deform_lat_r_base {parent=spine_03, loc=<2.5562 -10.3716 17.6918>, rot=<0 0 1 0>}
 deform_lat_r {parent=deform_lat_r_base, loc=<0 0 0>, rot=<0 0 0 -1>}
 Spine_metal_corr_dwn {parent=spine_03, loc=<3.2512 -23.9336 -8.9923>, rot=<0.0387 -0.706 -0.0387 -0.706>}
 Spine_metal_corr_up {parent=spine_03, loc=<19.6131 -21.0566 4.2436>, rot=<0.0325 -0.5927 -0.044 -0.8035>}
 thigh_l {parent=pelvis, loc=<-1.8716 -0.0109 -11.9936>, rot=<0.0012 0.0572 0.0155 -0.9982>}
 calf_l {parent=thigh_l, loc=<-41.9178 -0.2246 0.0926>, rot=<0.0006 -0.0161 -0.0391 -0.9991>}
 foot_l {parent=calf_l, loc=<-38.7057 1.2438 0.2452>, rot=<0.0975 -0.0474 0.0317 -0.9936>}
 ball_l {parent=foot_l, loc=<-6.1614 13.1469 0.6807>, rot=<0.0001 -0.0001 -0.7063 -0.7079>}
 Base_lape_l {parent=foot_l, loc=<-0.2111 5.2437 0.7861>, rot=<-0.0054 -0.775 0.632 0.0021>}
 Dyn_lape_l {parent=Base_lape_l, loc=<0 0 0>, rot=<0 0 0 1>}
 calf_twist_02_l {parent=calf_l, loc=<-15.9268 -0.0967 0.0397>, rot=<0.0237 0.0095 -0.0035 -0.9997>}
 calf_twist_01_l {parent=calf_l, loc=<-32.9335 -0.0967 0.0397>, rot=<0.0237 0.0095 -0.0035 -0.9997>}
 Snikers_corr_l {parent=calf_twist_01_l, loc=<-4.1455 -2.7002 0.0876>, rot=<-0.0439 -0.9973 -0.0266 0.0516>}
 deform_knee_in_base_l {parent=calf_l, loc=<1.298 -8.0356 -0.0496>, rot=<0 0 0 -1>}
 deform_knee_in_l {parent=deform_knee_in_base_l, loc=<0 0 0>, rot=<0 0 0 1>}
 deform_knee_out_base_l {parent=calf_l, loc=<0 6.8739 0>, rot=<0 0 0 -1>}
 deform_knee_out_l {parent=deform_knee_out_base_l, loc=<0 0 0>, rot=<0 0 0 1>}
 thigh_twist_01_l {parent=thigh_l, loc=<-21.6303 0 0>, rot=<0.0013 0 0 -1>}
 deform_groin_base_l {parent=thigh_l, loc=<-2.1908 8.5661 -1.7442>, rot=<0 0 0 -1>}
 deform_groin_l {parent=deform_groin_base_l, loc=<0 0 0>, rot=<0 0 0 1>}
 thigh_r {parent=pelvis, loc=<-1.8716 -0.0109 11.9936>, rot=<0.0572 -0.0012 0.9982 0.0155>}
 calf_r {parent=thigh_r, loc=<41.9178 0.2246 -0.0926>, rot=<0.0006 -0.0161 -0.0391 -0.9991>}
 foot_r {parent=calf_r, loc=<38.7057 -1.2438 -0.2452>, rot=<0.0975 -0.0474 0.0317 -0.9936>}
 ball_r {parent=foot_r, loc=<6.1614 -13.1469 -0.6807>, rot=<0.0001 -0.0001 -0.7063 -0.7079>}
 Base_lape_r {parent=foot_r, loc=<0.2111 -5.2437 -0.7861>, rot=<0.632 -0.0021 0.0054 -0.775>}
 Dyn_lape_r {parent=Base_lape_r, loc=<0 0 0>, rot=<0 0 0 1>}
 calf_twist_02_r {parent=calf_r, loc=<15.9268 0.0967 -0.0397>, rot=<0.0237 0.0095 -0.0035 -0.9997>}
 calf_twist_01_r {parent=calf_r, loc=<32.9335 0.0967 -0.0397>, rot=<0.0237 0.0095 -0.0035 -0.9997>}
 Snikers_corr_r {parent=calf_twist_01_r, loc=<4.1455 2.7002 -0.0876>, rot=<0.9973 -0.0439 -0.0516 -0.0266>}
 deform_knee_in_base_r {parent=calf_r, loc=<-1.2981 8.0356 0.0496>, rot=<0 0 0 -1>}
 deform_knee_in_r {parent=deform_knee_in_base_r, loc=<0 0 0>, rot=<0 0 0 1>}
 deform_knee_out_base_r {parent=calf_r, loc=<0 -6.8739 0>, rot=<0 0 0 -1>}
 deform_knee_out_r {parent=deform_knee_out_base_r, loc=<0 0 0>, rot=<0 0 0 1>}
 thigh_twist_01_r {parent=thigh_r, loc=<21.6303 0 0>, rot=<0.0013 0 0 -1>}
 deform_groin_base_r {parent=thigh_r, loc=<2.1908 -8.5661 1.7442>, rot=<0 0 0 -1>}
 deform_groin_r {parent=deform_groin_base_r, loc=<0 0 0>, rot=<0 0 0 -1>}
 deform_glute_base_l {parent=pelvis, loc=<-10.1877 -11.6217 -12.899>, rot=<0.0275 0.0501 -0.4472 -0.8926>}
 deform_glute_l {parent=deform_glute_base_l, loc=<0 0 0>, rot=<0 0 0 1>}
 deform_glute_base_r {parent=pelvis, loc=<-10.1877 -11.6218 12.899>, rot=<0.0501 -0.0275 0.8926 -0.4472>}
 deform_glute_r {parent=deform_glute_base_r, loc=<0 0 0>, rot=<0 0 0 1>}
 Base_Grenade_belt {parent=pelvis, loc=<7.2278 12.0465 -20.5593>, rot=<0.108 -0.887 0.4276 0.1367>}
 Grenade_A {parent=Base_Grenade_belt, loc=<-0.0592 1.6391 -8.5577>, rot=<0.127 0.0215 -0.0481 -0.9905>}
 Grenade_B {parent=Base_Grenade_belt, loc=<-0.5344 3.8031 -0.176>, rot=<0.0407 0.0172 -0.0498 -0.9978>}
 Grenade_C {parent=Base_Grenade_belt, loc=<-0.6898 2.8213 8.172>, rot=<-0.1049 0.0098 -0.0518 -0.9931>}
 hand_l_ctrl {controlType=EULER_TRANSFORM, type=Control, loc=<64.462 4.8638 101.1961>, rot=<0.6205 0.4613 -0.1793 0.6084>}
 PoleVectorl_l {controlType=EULER_TRANSFORM, type=Control, loc=<4.2721 -5.5702 131.7007>, rot=<-0.0008 -0.0336 0.0252 -0.9991>}
 PoleVector_r {controlType=EULER_TRANSFORM, type=Control, loc=<-4.2721 -5.5702 131.7007>, rot=<0.9991 0.0252 0.0336 -0.0008>}
 hand_r_ctrl {controlType=EULER_TRANSFORM, type=Control, loc=<-64.462 4.8638 101.1961>, rot=<-0.6084 -0.1793 -0.4613 0.6205>}
]
variables [
 BaseBallSliding {type=double, default=0}
 Crouch {type=double, default=0}
 GauntletsEquipt {type=bool, default=False}
 Get Velocity Z {type=double, default=0}
 IsPlayingEmote {type=bool, default=False}
 Melee {type=double, default=0}
 OpenParawithDelay {type=double, default=0}
 SIt {type=double, default=0}
 Two Arm up {type=double, default=0}
 Two Arms Hang {type=double, default=0}
]
