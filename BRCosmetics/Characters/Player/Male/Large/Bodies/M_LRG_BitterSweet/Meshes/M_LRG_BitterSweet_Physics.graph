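# Physics asset for the BitterSweet large male player body cosmetic, covering a small ring bone chain and the right upper arm.
# Dyn_ring carries the ring's only sized box; Base_ring and MSTR_ring are zero-extent anchor boxes.
# Base_ring rotates freely against MSTR_ring, while Dyn_ring's joint to Base_ring locks one swing axis.
# An unconstrained upperarm_r capsule supplies an arm collision shape, apparently for the ring to collide with.
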PhysicsAsset /BRCosmetics/Characters/Player/Male/Large/Bodies/M_LRG_BitterSweet/Meshes/M_LRG_BitterSweet_Physics
bodies {
 Base_ring [{shapeType=Box, extentX=0, extentY=0, extentZ=0, shapeName=Base_ring_box}]
 Dyn_ring [{
   shapeType=Box
   extentX=5.0739
   extentY=14.316
   extentZ=17.0704
   shapeName=Dyn_ring_box
   center=<2.7687 -0.8895 -6.8643>
   rotation=<rot 24.0251 19.9241 -88.7364>
  }]
 MSTR_ring [{shapeType=Box, extentX=0, extentY=0, extentZ=0, shapeName=MSTR_ring_box}]
 upperarm_r [{
   shapeType=Capsule
   length=15.9219
   radius=8.6304
   shapeName=upperarm_r_capsule
   center=<-10.6771 0.0812 0.1315>
   rotation=<rot 31.019 -78.3107 96.9281>
  }]
}
constraints [
 {
  bone1Name=Base_ring
  bone2Name=MSTR_ring
  swing1LimitDegrees=45
  swing1Motion=Free
  swing2LimitDegrees=45
  swing2Motion=Free
  twistLimitDegrees=45
  twistMotion=Free
 }
 {bone1Name=Dyn_ring, bone2Name=Base_ring, swing1LimitDegrees=5, swing1Motion=Locked, swing2LimitDegrees=45, twistLimitDegrees=5}
]
