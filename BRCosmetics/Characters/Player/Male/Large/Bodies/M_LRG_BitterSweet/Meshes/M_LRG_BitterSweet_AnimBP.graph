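# Animation Blueprint for the M_LRG_BitterSweet cosmetic male body, layering procedural secondary motion onto the attached parent's copied pose.
# A Control Rig takes glider, knee-slide, emote and gauntlet state, then RigidBody physics and Anim Dynamics on the zipper run before weighted constraints on zipper and ribbon base bones.
# Events feed those states; computed gravity, simulation space and head-rotation variables are not read by the AnimGraph.
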AnimBlueprint /BRCosmetics/Characters/Player/Male/Large/Bodies/M_LRG_BitterSweet/Meshes/M_LRG_BitterSweet_AnimBP
target_skeleton=&/M_LRG_BitterSweet_Skeleton
graphs [
 AnimGraph {nodes [
   n1 = Control Rig!(
    Source=Copy Pose From Mesh() {bUseAttachedParent=true}
    Alpha
    Two Arms Hang:double=Get HangTwo arms() -> (HangTwo arms:double)
    BaseBallSliding:double=Get BaseBallSlide() -> (BaseBallSlide:double)
    IsPlayingEmote:bool=Get bIsPlayingEmote()
    OpenParawithDelay:double=Get OpenParaschut() -> (OpenParaschut:double)
    GauntletsEquipt:bool=Get GauntlesEquip() -> (GauntlesEquip:bool)
   ) -> (Pose) {
    alphaBoolBlend {blendOption=Linear}
    controlRigAssetReference {blueprintRigClass=&/M_LRG_BitterSweet_CtrlRig.M_LRG_BitterSweet_CtrlRig_C}
   }
   n2 = RigidBody!(ComponentPose=Local To Component(LocalPose=@n1), Alpha, OverrideWorldGravity:Vector="0,0,-980") -> (Pose) {
    bClampLinearTranslationLimitToRefPose=true
    bOverrideWorldGravity=true
    overridePhysicsAsset=&/M_LRG_BitterSweet_Physics
    simulationSpace=BaseBoneSpace
    alphaBoolBlend {blendOption=Linear}
    baseBoneRef {boneName=upperarm_r}
    overrideWorldGravity {z=-980}
    simSpaceSettings {dampingAlpha=0, maxAngularAcceleration=40, maxAngularVelocity=8, maxLinearAcceleration=1100, maxLinearVelocity=800, worldAlpha=1}
   }
   n3 = Anim Dynamics - Bone: Dyn_zipper!(SimulationSpace:AnimPhysSimSpaceType Enum=World, ComponentPose=@n2, Alpha=1, GravityOverride:Vector="0,0,-980", ExternalForce="0,0,0") -> (Pose) {
    angularDampingOverride=1
    angularSpringConstant=500
    bAngularSpring=true
    bOverrideAngularDamping=true
    bUseGravityOverride=true
    gravityScale=0.35
    simulationSpace=World
    alphaBoolBlend {blendOption=Linear}
    boundBone {boneName=Dyn_zipper}
    componentAppliedLinearAccClamp=<6000 6000 6000>
    componentLinearAccScale=<1 1 1>
    componentLinearVelScale=<1 1 1>
    gravityOverride {z=-980}
    physicsBodyDefinitions [{
      boundBone {boneName=Dyn_zipper}
      boxExtents=<10 10 10>
      constraintSetup {angularLimitsMax=<10 60 12>, angularLimitsMin=<-5 -3 -12>, angularTarget=<1 0 0>}
      localJointOffset=<0.7 -0.3 0.7>
     }]
   }
   n4 = "Constraint - Base_zipper (PPPPPPPPPP)"!(
    ComponentPose=@n3
    Alpha=1
    ConstraintWeights_0=0.0068
    ConstraintWeights_1:float=0.0248
    ConstraintWeights_2:float=0.0719
    ConstraintWeights_3:float=0.56
    ConstraintWeights_4:float=0.004
    ConstraintWeights_5:float=0.0039
    ConstraintWeights_6:float=0.2956
    ConstraintWeights_7:float=0.0314
    ConstraintWeights_8:float=0.0008
    ConstraintWeights_9:float=0.0008
   ) -> (Pose) {
    alphaBoolBlend {blendOption=Linear}
    boneToModify {boneName=Base_zipper}
    constraintSetup [
     {targetBone {boneName=spine_02}}
     {targetBone {boneName=spine_03}}
     {targetBone {boneName=spine_04}}
     {targetBone {boneName=spine_05}}
     {targetBone {boneName=clavicle_l}}
     {targetBone {boneName=clavicle_r}}
     {targetBone {boneName=neck_01}}
     {targetBone {boneName=neck_02}}
     {targetBone {boneName=deform_pec_r}}
     {targetBone {boneName=deform_pec_l}}
    ]
    constraintWeights [1,1,1,1,1,1,1,1,1,1]
   }
   n5 = "Constraint - Base_ribbon_BK (PPPPPPPP)"!(
    ComponentPose=@n4
    Alpha=1
    ConstraintWeights_0=0.0001
    ConstraintWeights_1:float=0
    ConstraintWeights_2:float=0.0001
    ConstraintWeights_3:float=0.0001
    ConstraintWeights_4:float=0
    ConstraintWeights_5:float=0
    ConstraintWeights_6:float=0.3826
    ConstraintWeights_7:float=0.6172
   ) -> (Pose) {
    alphaBoolBlend {blendOption=Linear}
    boneToModify {boneName=Base_ribbon_BK}
    constraintSetup [
     {targetBone {boneName=spine_03}}
     {targetBone {boneName=spine_04}}
     {targetBone {boneName=spine_05}}
     {targetBone {boneName=clavicle_r}}
     {targetBone {boneName=upperarm_twist_01_r}}
     {targetBone {boneName=deform_upperarm_bk_r}}
     {targetBone {boneName=Base_chick__B}}
     {targetBone {boneName=Base_chick__A}}
    ]
    constraintWeights [1,1,1,1,1,1,1,1]
   }
   n6 = "Constraint - Base_ribbon_FR (PPPPPPPPPP)"!(
    ComponentPose=@n5
    Alpha=1
    ConstraintWeights_0=0.0291
    ConstraintWeights_1:float=0.01
    ConstraintWeights_2:float=0.0254
    ConstraintWeights_3:float=0.2412
    ConstraintWeights_4:float=0.001
    ConstraintWeights_5:float=0.0102
    ConstraintWeights_6:float=0.0003
    ConstraintWeights_7:float=0.0476
    ConstraintWeights_8:float=0.1046
    ConstraintWeights_9:float=0.5307
   ) -> (Pose) {
    alphaBoolBlend {blendOption=Linear}
    boneToModify {boneName=Base_ribbon_FR}
    constraintSetup [
     {targetBone {boneName=spine_03}}
     {targetBone {boneName=spine_04}}
     {targetBone {boneName=spine_05}}
     {targetBone {boneName=clavicle_r}}
     {targetBone {boneName=deform_upperarm_fr_r}}
     {targetBone {boneName=deform_upperarm_up_r}}
     {targetBone {boneName=upperarm_twist_02_r}}
     {targetBone {boneName=deform_pec_r}}
     {targetBone {boneName=Base_chick__B}}
     {targetBone {boneName=Base_chick__A}}
    ]
    constraintWeights [1,1,1,1,1,1,1,1,1,1]
   }
   Output Pose(Result=Component To Local(ComponentPose=@n6)) {layerGroup=None}
  ]}
 EventGraph {nodes [
   Event BlueprintUpdateAnimation()
   n1 = GetOwnerSkeletalMeshComponent()
   n2 = BreakRotator(InRot=BreakTransform(InTransform=GetDeltaTransformFromRefPose(self=@n1, BoneName=head, BaseName=neck_02)).Rotation)
   n3 = Get Z percentageEmote() -> (Z percentageEmote:float)
   n4 = Get X percentageEmote() -> (X percentageEmote:float)
   n5 = Get Y percentageEmote() -> (Y percentageEmote:float)
   > Set headRotLimit(headRotLimit:Rotator=MakeRotator(
     Roll=float * float!(A=@n2, B=-1, C:double=@n4) -> (ReturnValue)
     Pitch=float * float(A=@n2.Pitch, B=-1, C=@n5)
     Yaw=float * float(A=@n2.Yaw, B=-1, C=@n3)
    )) -> (Output_Get:Rotator)
   > n6 = Gravity Override State Function DEPRECATED(CharacterPartAnimInstance=Self-Reference(), GameTimeInSeconds=GetGameTimeInSeconds(), GravityOverrideParams=Get Gravity Override Params())
   > Set Gravity Override(Gravity Override=@n6)
   > n7 = Branch(Condition=Get bIsInFrontEnd())
   > Set SimlationSpace(SimlationSpace:AnimPhysSimSpaceType Enum=World) -> (Output_Get:AnimPhysSimSpaceType Enum)
   Set SimlationSpace(execute=@n7.else, SimlationSpace=Component)
   Event BlueprintInitializeAnimation()
   n8 = TryGetPawnOwner()
   > Is Valid(InputObject=@n8)
   > n9 = Cast To FortPlayerPawnAthena(Object=@n8)
   > Set As Fort Player Pawn Athena(As Fort Player Pawn Athena:Fort Player Pawn Athena Object Reference=@n9) -> (Output_Get:Fort Player Pawn Athena Object Reference)
   n10 = Event OnCharacterCustomizationLoaded()
   > Set GauntlesEquip(GauntlesEquip:bool=IsValid(Object=Get Array Item with EFort Custom Part Type(Array:Array of Custom Character Part Object References=@n10.CustomParts, PartType=Gameplay) -> (Output:Custom Character Part Object Reference))) -> (Output_Get:bool)
   Event OnBeginKneeSliding() {is=title K2Node_Event}
   > Set BaseBallSlide(BaseBallSlide:double=1) -> (Output_Get:double)
   Event OnParachuteClosed()
   > Set HangTwo arms(HangTwo arms:double=0) -> (Output_Get:double)
   Event OnEndKneeSliding() {is=title K2Node_Event}
   > Set BaseBallSlide(BaseBallSlide=0)
   Event OnParachuteOpened()
   > Switch on EGliderType(Selection=Get PlayerGliderType() -> (PlayerGliderType=HangGlider))
   > Set HangTwo arms(HangTwo arms=1)
   > Set OpenParaschut(OpenParaschut:double=1) -> (Output_Get:double)
   > Delay(Duration=2)
   > Set OpenParaschut(OpenParaschut=0)
   n11 = Get As Fort Player Pawn Athena() -> (As Fort Player Pawn Athena:Fort Player Pawn Athena Object Reference)
   Is Valid(InputObject=@n11)
   n12 = GetDisplayName(Object=GetSkeletalMeshForPartType(self=@n11, PartType=Gameplay))
   > Set GauntlesEquip(GauntlesEquip=OR Boolean!(
     A="Equal Exactly (String)"(A=@n12, B=PlayerPawn_Athena0.CharacterPartSkelMesh_Gameplay_0 GauntletsTemplate_Gadget)
     B="Equal Exactly (String)"(A=@n12, B=PlayerPawn_Athena0.CharacterPartSkelMesh_Gameplay_0 SprocketPoppy_Gauntlet)
     C:bool="Equal Exactly (String)"(A=@n12, B=PlayerPawn_Athena0.CharacterPartSkelMesh_Gameplay_0 CoffeePuff_Gauntlet)
    ) -> (ReturnValue))
   "To String (Boolean)"(InBool=Set GauntlesEquip(GauntlesEquip=false))
   PrintString!(execute, InString="To String (Boolean)"(InBool=false), bPrintToScreen, bPrintToLog, TextColor, Duration, Key) -> (then)
   "Equal Exactly (String)"(A=GetDisplayName(), B=B_Athena_Pickaxe_DualWield_Forearm_Generic1)
   GetPartSkeletalMeshComponent()
   "Equal (Object)"(B=/Game/Weapons/FORT_Melee/Blueprints/B_Athena_Pickaxe_DualWield_Forearm_Generic)
   Get CurrentWeapon() -> (CurrentWeapon:Fort Weapon Object Reference)
  ]}
]
variables [
 As Fort Player Pawn Athena {type=FortPlayerPawnAthena}
 BaseBallSlide {type=number}
 GauntlesEquip {type=boolean}
 Gravity Override {type=Vector}
 Gravity Override Params {type=GravityOverrideParamsStruct}
 HangTwo arms {type=number}
 headRotLimit {type=Rotator}
 Is Targeting {type=boolean}
 Melee {type=number}
 OpenParaschut {type=number}
 SimlationSpace {type=AnimPhysSimSpaceType}
 X percentageEmote {type=number, default=1}
 Y percentageEmote {type=number, default=1}
 Z percentageEmote {type=number, default=1}
]
