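# Physics asset for the large male BitterSweet cosmetic body, named for its cape and likely serving as that cape's collision body.
# Single capsules cover pelvis, spine, clavicles, arms, hands, legs and feet; no cape bones have bodies.
# Extra capsules sit on Base_chick__B, linked to the right clavicle, and on only the right thigh twist.
# Constraints share uniform swing and twist limits, with forearms and upper-arm twists linked directly to clavicles.
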PhysicsAsset /BRCosmetics/Characters/Player/Male/Large/Bodies/M_LRG_BitterSweet/Meshes/M_LRG_BitterSweet_Physic_Cape
bodies {
 Base_chick__B [{shapeType=Capsule, length=13.2451, radius=10.1693, center=<-0.0775 0.0334 1.2623>, rotation=<rot 3.2097 9.6231 -87.9196>}]
 calf_l [{shapeType=Capsule, length=23.0708, radius=8.786, center=<-12.1732 -1.4956 0.1756>, rotation=<rot 0 -82.9958 -89.1795>}]
 calf_r [{shapeType=Capsule, length=23.0708, radius=8.786, center=<12.1732 1.4956 -0.1756>, rotation=<rot 0 -82.9958 90.8205>}]
 clavicle_l [{shapeType=Capsule, length=5.0294, radius=11.1735, center=<9.3068 0 4.7018>, rotation=<rot 85 -179.9984 -179.9984>}]
 clavicle_r [{shapeType=Capsule, length=5.0294, radius=11.1735, center=<-9.3068 0 -4.7018>, rotation=<rot 85 -179.9984 0.0016>}]
 foot_l [{shapeType=Capsule, length=14.1882, radius=9.8814, center=<-4.9062 5.6662 0.7006>, rotation=<rot 0 -167.6323 -83.9274>}]
 foot_r [{shapeType=Capsule, length=14.1882, radius=9.8814, center=<4.9062 -5.6662 -0.7006>, rotation=<rot 0 -167.6323 96.0726>}]
 hand_l [{shapeType=Capsule, length=9.054, radius=7.1668, rotation=<rot 0 82.7419 -67.8397>}]
 hand_r [{shapeType=Capsule, length=9.054, radius=7.1668, rotation=<rot 0 82.7419 -67.8397>}]
 lowerarm_l [{shapeType=Capsule, length=19.2951, radius=8.2033, center=<9.1881 0.85 0.6126>, rotation=<rot 0 95.2855 -86.2015>}]
 lowerarm_r [{shapeType=Capsule, length=19.2951, radius=8.2033, center=<-9.1881 -0.85 -0.6126>, rotation=<rot 0 95.2855 93.7985>}]
 pelvis [{shapeType=Capsule, length=6.9487, radius=26.9904}]
 spine_02 [{shapeType=Capsule, length=24.0502, radius=21.9207, rotation=<rot 6.2789 -90 -0.3674>}]
 spine_03 [{shapeType=Capsule, length=26.1444, radius=24.6936, center=<6.1931 0 0>}]
 spine_04 [{shapeType=Capsule, length=26.1004, radius=22.7635, center=<0 -3.8478 0.6674>}]
 spine_05 [{shapeType=Capsule, length=20.796, radius=20.7303, center=<0 -3.5982 0>}]
 thigh_l [{shapeType=Capsule, length=19.8024, radius=12.1549, center=<-10.7062 -0.2387 2.0059>, rotation=<rot 0 -88.723 -89.3908>}]
 thigh_r [{shapeType=Capsule, length=19.8024, radius=12.1549, center=<10.7062 0.2387 -2.0059>, rotation=<rot 0 -88.723 90.6092>}]
 thigh_twist_01_r [{shapeType=Capsule, length=14.8826, radius=14.7267, rotation=<rot -36.0844 -90 -56.5269>}]
 upperarm_twist_01_l [{shapeType=Capsule, length=21.1701, radius=9.9514, center=<0 -1.6114 0>, rotation=<rot -85 179.9983 -179.9982>}]
 upperarm_twist_01_r [{shapeType=Capsule, length=21.1701, radius=9.9514, center=<0 1.6114 0>, rotation=<rot -85 179.9983 0.0018>}]
}
constraints [
 {bone1Name=spine_02, bone2Name=pelvis, swing1LimitDegrees=45, swing2LimitDegrees=45, twistLimitDegrees=45}
 {bone1Name=spine_03, bone2Name=spine_02, swing1LimitDegrees=45, swing2LimitDegrees=45, twistLimitDegrees=45}
 {bone1Name=spine_04, bone2Name=spine_03, swing1LimitDegrees=45, swing2LimitDegrees=45, twistLimitDegrees=45}
 {bone1Name=spine_05, bone2Name=spine_04, swing1LimitDegrees=45, swing2LimitDegrees=45, twistLimitDegrees=45}
 {bone1Name=clavicle_l, bone2Name=spine_05, swing1LimitDegrees=45, swing2LimitDegrees=45, twistLimitDegrees=45}
 {bone1Name=lowerarm_l, bone2Name=clavicle_l, swing1LimitDegrees=45, swing2LimitDegrees=45, twistLimitDegrees=45}
 {bone1Name=hand_l, bone2Name=lowerarm_l, swing1LimitDegrees=45, swing2LimitDegrees=45, twistLimitDegrees=45}
 {bone1Name=upperarm_twist_01_l, bone2Name=clavicle_l, swing1LimitDegrees=45, swing2LimitDegrees=45, twistLimitDegrees=45}
 {bone1Name=clavicle_r, bone2Name=spine_05, swing1LimitDegrees=45, swing2LimitDegrees=45, twistLimitDegrees=45}
 {bone1Name=lowerarm_r, bone2Name=clavicle_r, swing1LimitDegrees=45, swing2LimitDegrees=45, twistLimitDegrees=45}
 {bone1Name=hand_r, bone2Name=lowerarm_r, swing1LimitDegrees=45, swing2LimitDegrees=45, twistLimitDegrees=45}
 {bone1Name=upperarm_twist_01_r, bone2Name=clavicle_r, swing1LimitDegrees=45, swing2LimitDegrees=45, twistLimitDegrees=45}
 {bone1Name=Base_chick__B, bone2Name=clavicle_r, swing1LimitDegrees=45, swing2LimitDegrees=45, twistLimitDegrees=45}
 {bone1Name=thigh_l, bone2Name=pelvis, swing1LimitDegrees=45, swing2LimitDegrees=45, twistLimitDegrees=45}
 {bone1Name=calf_l, bone2Name=thigh_l, swing1LimitDegrees=45, swing2LimitDegrees=45, twistLimitDegrees=45}
 {bone1Name=foot_l, bone2Name=calf_l, swing1LimitDegrees=45, swing2LimitDegrees=45, twistLimitDegrees=45}
 {bone1Name=calf_r, bone2Name=thigh_r, swing1LimitDegrees=45, swing2LimitDegrees=45, twistLimitDegrees=45}
 {bone1Name=foot_r, bone2Name=calf_r, swing1LimitDegrees=45, swing2LimitDegrees=45, twistLimitDegrees=45}
 {bone1Name=thigh_twist_01_r, bone2Name=thigh_r, swing1LimitDegrees=45, swing2LimitDegrees=45, twistLimitDegrees=45}
 {bone1Name=thigh_r, bone2Name=pelvis, swing1LimitDegrees=45, swing2LimitDegrees=45, twistLimitDegrees=45}
]
