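# Physics asset for the leaf face accessory of the large male BitterSweet player character, with one capsule body per bone.
# Twenty Dyn_leaves bodies each attach directly to Base_leaves in a star layout, and Base_leaves attaches to MSTR_leaves.
# All constraints lock twist; leaf swing limits are narrow on one axis and wider on the other, with axes swapped for K, L and j.
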PhysicsAsset /BRCosmetics/Characters/Player/Male/Large/Bodies/M_LRG_BitterSweet/Meshes/Parts/M_LRG_BitterSweet_FaceAcc_Physics
bodies {
 Base_leaves [{shapeType=Capsule, length=0, radius=4.7624, shapeName=Base_leaves_capsule, center=<0 0.1 0>}]
 Dyn_leaves_A [{shapeType=Capsule, length=2, radius=2, shapeName=Dyn_leaves_A_capsule, center=<3 0 0>}]
 Dyn_leaves_B [{shapeType=Capsule, length=2, radius=2, shapeName=Dyn_leaves_B_capsule, center=<3 0 0>}]
 Dyn_leaves_C [{shapeType=Capsule, length=2, radius=2, shapeName=Dyn_leaves_C_capsule, center=<0 4.5512 1.8424>, rotation=<rot 0.0002 0 100>}]
 Dyn_leaves_D [{shapeType=Capsule, length=2, radius=2, shapeName=Dyn_leaves_D_capsule, center=<3.0599 -0.1994 0>}]
 Dyn_leaves_E [{shapeType=Capsule, length=2, radius=2, shapeName=Dyn_leaves_E_capsule, center=<3 0 0>}]
 Dyn_leaves_F [{shapeType=Capsule, length=2, radius=2, shapeName=Dyn_leaves_F_capsule, center=<3 0 0>}]
 Dyn_leaves_G [{shapeType=Capsule, length=2, radius=2, shapeName=Dyn_leaves_G_capsule, center=<3 0 0>}]
 Dyn_leaves_H [{shapeType=Capsule, length=2, radius=2, shapeName=Dyn_leaves_H_capsule, center=<3 0 0>}]
 Dyn_leaves_I [{shapeType=Capsule, length=0, radius=2, shapeName=Dyn_leaves_I_capsule, center=<3.815 0 4.6099>}]
 Dyn_leaves_K [{shapeType=Capsule, length=2, radius=2, shapeName=Dyn_leaves_K_capsule, center=<6.1946 0 0>}]
 Dyn_leaves_L [{shapeType=Capsule, length=2, radius=2, shapeName=Dyn_leaves_L_capsule, center=<5.6743 0 0>}]
 Dyn_leaves_O [{shapeType=Capsule, length=2, radius=2, shapeName=Dyn_leaves_O_capsule, center=<3 0 0>}]
 Dyn_leaves_Q [{shapeType=Capsule, length=2, radius=2, shapeName=Dyn_leaves_Q_capsule, center=<5.0085 0 -0.9435>}]
 Dyn_leaves_R [{shapeType=Capsule, length=2, radius=2, shapeName=Dyn_leaves_R_capsule, center=<4.4515 0 0>}]
 Dyn_leaves_T [{shapeType=Capsule, length=2, radius=2, shapeName=Dyn_leaves_T_capsule, center=<2 0 0>}]
 Dyn_leaves_U [{shapeType=Capsule, length=2, radius=2, shapeName=Dyn_leaves_U_capsule, center=<3 0 0>}]
 Dyn_leaves_W [{shapeType=Capsule, length=2, radius=2, shapeName=Dyn_leaves_W_capsule, center=<4.0319 0 -0.2517>, rotation=<rot 70.0002 -0.0001 0>}]
 Dyn_leaves_Y [{shapeType=Capsule, length=2, radius=2, shapeName=Dyn_leaves_Y_capsule, center=<3 0 0>}]
 Dyn_leaves_Z [{shapeType=Capsule, length=2, radius=2, shapeName=Dyn_leaves_Z_capsule, center=<3.3301 0 0>, rotation=<rot 75.0001 -0.0004 -0.0003>}]
 Dyn_leaves_j [{shapeType=Capsule, length=2, radius=2, shapeName=Dyn_leaves_j_capsule, center=<5.3193 -0.3766 0>}]
 MSTR_leaves [{shapeType=Capsule, length=0.505, radius=0.505, shapeName=MSTR_leaves_capsule, rotation=<rot 0 -90 0>}]
}
constraints [
 {bone1Name=Base_leaves, bone2Name=MSTR_leaves, swing1LimitDegrees=1, swing2LimitDegrees=1, twistLimitDegrees=45, twistMotion=Locked}
 {bone1Name=Dyn_leaves_A, bone2Name=Base_leaves, swing1LimitDegrees=3, swing2LimitDegrees=25, twistLimitDegrees=45, twistMotion=Locked}
 {bone1Name=Dyn_leaves_B, bone2Name=Base_leaves, swing1LimitDegrees=5, swing2LimitDegrees=20, twistLimitDegrees=45, twistMotion=Locked}
 {bone1Name=Dyn_leaves_C, bone2Name=Base_leaves, swing1LimitDegrees=3, swing2LimitDegrees=25, twistLimitDegrees=45, twistMotion=Locked}
 {bone1Name=Dyn_leaves_D, bone2Name=Base_leaves, swing1LimitDegrees=5, swing2LimitDegrees=20, twistLimitDegrees=45, twistMotion=Locked}
 {bone1Name=Dyn_leaves_E, bone2Name=Base_leaves, swing1LimitDegrees=4, swing2LimitDegrees=26, twistLimitDegrees=45, twistMotion=Locked}
 {bone1Name=Dyn_leaves_F, bone2Name=Base_leaves, swing1LimitDegrees=5, swing2LimitDegrees=25, twistLimitDegrees=45, twistMotion=Locked}
 {bone1Name=Dyn_leaves_G, bone2Name=Base_leaves, swing1LimitDegrees=3, swing2LimitDegrees=27, twistLimitDegrees=45, twistMotion=Locked}
 {bone1Name=Dyn_leaves_H, bone2Name=Base_leaves, swing1LimitDegrees=3, swing2LimitDegrees=19, twistLimitDegrees=45, twistMotion=Locked}
 {bone1Name=Dyn_leaves_I, bone2Name=Base_leaves, swing1LimitDegrees=3, swing2LimitDegrees=20, twistLimitDegrees=45, twistMotion=Locked}
 {bone1Name=Dyn_leaves_K, bone2Name=Base_leaves, swing1LimitDegrees=20, swing2LimitDegrees=5, twistLimitDegrees=45, twistMotion=Locked}
 {bone1Name=Dyn_leaves_L, bone2Name=Base_leaves, swing1LimitDegrees=20, swing2LimitDegrees=5, twistLimitDegrees=45, twistMotion=Locked}
 {bone1Name=Dyn_leaves_O, bone2Name=Base_leaves, swing1LimitDegrees=5, swing2LimitDegrees=25, twistLimitDegrees=45, twistMotion=Locked}
 {bone1Name=Dyn_leaves_Q, bone2Name=Base_leaves, swing1LimitDegrees=5, swing2LimitDegrees=20, twistLimitDegrees=45, twistMotion=Locked}
 {bone1Name=Dyn_leaves_R, bone2Name=Base_leaves, swing1LimitDegrees=5, swing2LimitDegrees=20, twistLimitDegrees=45, twistMotion=Locked}
 {bone1Name=Dyn_leaves_T, bone2Name=Base_leaves, swing1LimitDegrees=5, swing2LimitDegrees=23, twistLimitDegrees=45, twistMotion=Locked}
 {bone1Name=Dyn_leaves_U, bone2Name=Base_leaves, swing1LimitDegrees=6, swing2LimitDegrees=26, twistLimitDegrees=45, twistMotion=Locked}
 {bone1Name=Dyn_leaves_W, bone2Name=Base_leaves, swing1LimitDegrees=3, swing2LimitDegrees=28, twistLimitDegrees=45, twistMotion=Locked}
 {bone1Name=Dyn_leaves_Y, bone2Name=Base_leaves, swing1LimitDegrees=3, swing2LimitDegrees=25, twistLimitDegrees=45, twistMotion=Locked}
 {bone1Name=Dyn_leaves_Z, bone2Name=Base_leaves, swing1LimitDegrees=3, swing2LimitDegrees=27, twistLimitDegrees=45, twistMotion=Locked}
 {bone1Name=Dyn_leaves_j, bone2Name=Base_leaves, swing1LimitDegrees=20, swing2LimitDegrees=5, twistLimitDegrees=45, twistMotion=Locked}
]
